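# Niagara system for the RelayStickBounty female player skin, drawing glowing ribbons and small lights attached to the character's body.
# Ribbon emitters (both ArmSplines variants are disabled) burst-spawn grouped points, which FN_BoneFlow wraps along skeletal mesh sockets, while BodyLights pins colored lights to sockets.
# System amplitude, timeline and reactive values drive wave, width, color and material response.
# Time of day tints or fades effects, and quality level adjusts ribbon opacity.
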NiagaraSystem /BRCosmetics/Characters/Player/Female/Medium/Bodies/F_MED_RelayStickBounty/FX/NS_RelayStickBounty_Skin
dependencies {
 usedDataInterfaces [
  &NiagaraDataInterfaceColorCurve
  &NiagaraDataInterfaceArrayInt32
  &NiagaraDataInterfaceArrayFloat3
  &NiagaraDataInterfaceCurve
  &NiagaraDataInterfaceArrayFloat
  &NiagaraDataInterfaceSkeletalMesh
 ]
 usedDynamicInputs [
  &FNTimeOfDayScalar
  &FN_FrontEndFloatModifier
  &Power
  &SetBoolByFloatComparison
 ]
 usedModules [
  &self:SystemSpawnScript.SystemScriptSource.SystemScriptGraph.NiagaraNodeAssignment_2.SetVariables_#1
  &self:SystemSpawnScript.SystemScriptSource.SystemScriptGraph.NiagaraNodeAssignment_8.SetVariables_#2
  &self:SystemSpawnScript.SystemScriptSource.SystemScriptGraph.NiagaraNodeAssignment_3.SetVariables_#3
  &SystemState
  &self:SystemSpawnScript.SystemScriptSource.SystemScriptGraph.NiagaraNodeAssignment_0.SetVariables_#4
  &self:SystemSpawnScript.SystemScriptSource.SystemScriptGraph.NiagaraNodeAssignment_7.SetVariables_#5
  &self:SystemSpawnScript.SystemScriptSource.SystemScriptGraph.NiagaraNodeAssignment_9.SetVariables_#6
  &Timeline
 ]
}
emitters {
 ArmSplines {
  bEnabled=false
  emitterUpdateScript [
   &EmitterState(Life Cycle Mode=::NewEnumerator1, Loop Behavior=::NewEnumerator1, Loop Duration Mode=::NewEnumerator1, UseLoopDelay=false)
   &SpawnBurst_Instantaneous(Spawn Count=36, Use Spawn Probability=false, UseLoopCountLimit=false)
  ]
  particleSpawnScript [
   &PartitionParticles(
    Normalized Index In Partition Includes One=true
    Number Of Partitions=6
    Particles Per Partition=6
    Partition Mode:ENiagara_PartitionMode=::NewEnumerator1
    Write Partition to Ribbon ID=::NewEnumerator0
   )
   &InitializeParticle(
    Color=<rgba 1 1 1 0.33>
    Lifetime=5
    Ribbon Width=7
    Ribbon Width Mode=::NewEnumerator1
    Sprite Size Mode=::NewEnumerator3
    Uniform Sprite Size=2
    UsePositionOffset=false
    Write Material Random=true
   )
  ]
  particleUpdateScript [
   /SetVariables(Particles.QualityLevelOpacity:NiagaraFloat=&Lerp_Float(
     A=1
     B=&FloatFromCurve(
      FloatCurve=DataInterface() {propertyValues {Curve {keys [
          {interpMode=RCIM_Cubic, tangentMode=RCTM_User, time=0.05, value=0}
          {arriveTangent=1.4286, time=0.3, value=1}
          {interpMode=RCIM_Cubic, tangentMode=RCTM_User, time=0.75, value=1}
          {tangentMode=RCTM_User, time=1, value=0}
         ]}}}
      CurveIndex=$Particles.PartitionParticles.Normalized Index In Partition
     )
     Alpha=&MakeFloatFromInt(Int=&SelectIntFromArray(Direct Array Index=$Engine.QualityLevel, Int Selection Array=DataInterface() {propertyValues {IntData [1,1,0,0,0]}}))
    )) {isSetParameters=true, module=SetVariables_#7}
   /SetVariables(Particles.Wave:NiagaraFloat=&Waveform(
     Waveform Count=::NewEnumerator1
     Global Input Position=&FloatFromCurve(
      FloatCurve=DataInterface() {propertyValues {Curve {keys [{tangentMode=RCTM_User, time=0, value=0}, {tangentMode=RCTM_User, time=1, value=1}]}}}
      CurveIndex=$Particles.PartitionParticles.Normalized Index In Partition
      Scale Curve=&Multiply_Float(
       A=1
       B=&FloatFromCurve(
        FloatCurve=DataInterface() {propertyValues {Curve {keys [
            {interpMode=RCIM_Cubic, tangentMode=RCTM_User, time=0.1, value=0}
            {arriveTangent=1.1111, time=0.2, value=1}
            {interpMode=RCIM_Cubic, tangentMode=RCTM_User, time=0.8, value=1}
            {tangentMode=RCTM_User, time=0.9, value=0}
           ]}}}
        CurveIndex=$Particles.PartitionParticles.Normalized Index In Partition
       )
      )
     )
     "Waveform [1]"=::NewEnumerator0
     "[1] Phase Scale"=0.25
     "[1] Frequency"=1.5
     "[1] Amplitude Min/Max"=<0 1>
     "[1] Scale"=0.5
     "[1] Bias"=1
     "[2] Average"=false
     "[2] Input Position"=0
     "[2] Phase"=$Emitter.Age
     "[2] Phase Scale"=-0.5
     "[2] Compound Frequency"=2
     "[2] Frequency"=2
     "[2] Amplitude Min/Max"=<0 1>
     "[2] Scale"=&Lerp_Float(A=0, B=1, Alpha=$System.AmplitudeValue)
     "[2] Bias"=0
     "[2] Exponent"=1
     "[2] Pulse Duration"=0.1
     "[2] Clip Triangle Min/Max"=<0 0>
     "[2] Random Seed"=0
    )) {isSetParameters=true, module=SetVariables_#8}
   &ScaleRibbonWidth(Ribbon Width Scale=&FloatFromCurve(
     FloatCurve=DataInterface() {propertyValues {Curve {keys [
         {arriveTangent=6.0891, interpMode=RCIM_Cubic, leaveTangent=6.0891, tangentMode=RCTM_User, time=0, value=1}
         {arriveTangent=1.25, time=0.1, value=1}
         {interpMode=RCIM_Cubic, tangentMode=RCTM_User, time=0.9, value=1}
         {arriveTangent=-4.9474, tangentMode=RCTM_User, time=1, value=1}
        ]}}}
     CurveIndex=$Particles.Initial.PartitionParticles.Normalized Index In Partition
     Scale Curve=&Lerp_Float(A=1, B=1.25, Alpha=$System.AmplitudeValue)
    ))
   /SetVariables(Particles.GlowRibbonWidth:NiagaraFloat=&Multiply_Float(
     A=$Particles.Wave
     B=&Lerp_Float(
      A=&FloatFromCurve(
       FloatCurve=DataInterface() {propertyValues {Curve {keys [
           {interpMode=RCIM_Cubic, tangentMode=RCTM_User, time=0, value=1}
           {arriveTangent=-2.0527, tangentMode=RCTM_User, time=0.2, value=0.6}
           {arriveTangent=1.9371, interpMode=RCIM_Cubic, leaveTangent=1.9371, tangentMode=RCTM_User, time=0.8, value=0.6}
           {tangentMode=RCTM_User, time=1, value=1}
          ]}}}
       CurveIndex=$Particles.PartitionParticles.Normalized Index In Partition
       Scale Curve=4
      )
      B=4
      Alpha=$System.AmplitudeValue
     )
    )) {isSetParameters=true, module=SetVariables_#9}
   &ParticleState(
    Kill Particles When Lifetime Has Elapsed=false
    Let Infinitely Lived Particles Die When Emitter Deactivates=true
    Loop Particles Lifetime=true
   )
   &FN_BoneFlow(
    Angle:NiagaraFloat=&Add_Float(
     A=&SelectFloatFromArray(
      Direct Array Index=$Particles.PartitionParticles.Partition
      Float Selection Array=DataInterface() {propertyValues {FloatData [0.25,0.25,0.25,0.25,0.225,0.13]}}
     )
     B=&FloatFromCurve(
      FloatCurve=DataInterface() {propertyValues {Curve {keys [{interpMode=RCIM_Cubic, tangentMode=RCTM_User, time=0, value=0}, {interpMode=RCIM_Cubic, tangentMode=RCTM_User, time=0.8, value=1}]}}}
      CurveIndex=$Particles.PartitionParticles.Normalized Index In Partition
      Scale Curve=&SelectFloatFromArray(
       Direct Array Index=$Particles.PartitionParticles.Partition
       Float Selection Array=DataInterface() {propertyValues {FloatData [0,0,0,0,-0.3,0.4]}}
      )
     )
    )
    BoneID:NiagaraInt32=$Particles.Initial.PartitionParticles.Partition
    Enable Position Offsets=true
    Height:NiagaraFloat=$Particles.PartitionParticles.Normalized Index In Partition
    Offset End:Vector3f=&SelectVectorFromArray(
     Direct Array Index=$Particles.PartitionParticles.Partition
     Vector Selection Array=DataInterface() {propertyValues {FloatData [<0 0 2>, <0 0 -2>, <0 0 0>, <0 0 0>, <0 0 0>, <0 0 0>]}}
    )
    Offset Mid:Vector3f=&SelectVectorFromArray(
     Direct Array Index=$Particles.PartitionParticles.Partition
     Vector Selection Array=DataInterface() {propertyValues {FloatData [<0 0 0>, <0 0 0>, <0 0 0>, <0 0 0>, <0 0 0>, <0 0 0>]}}
    )
    Offset Start:Vector3f=&SelectVectorFromArray(
     Direct Array Index=$Particles.PartitionParticles.Partition
     Vector Selection Array=DataInterface() {propertyValues {FloatData [<0 0 0>, <0 0 0>, <0 0 0>, <0 0 0>, <-2.5 0 -2>, <-2.5 0 -2>]}}
    )
    Radius:NiagaraFloat=&Multiply_Float(
     A=&MaskFloatBySpawnGroup(
      Masked Float=&FloatFromCurve(
       FloatCurve=DataInterface() {propertyValues {Curve {keys [
           {arriveTangent=13.5458, interpMode=RCIM_Cubic, leaveTangent=13.5458, tangentMode=RCTM_User, time=0, value=0.5}
           {interpMode=RCIM_Cubic, tangentMode=RCTM_User, time=0.5, value=1}
           {arriveTangent=-13.0583, tangentMode=RCTM_User, time=1, value=0.5}
          ]}}}
       CurveIndex=$Particles.PartitionParticles.Normalized Index In Partition
      )
      Passthrough Float=&FloatFromCurve(
       FloatCurve=DataInterface() {propertyValues {Curve {keys [
           {arriveTangent=21.0298, interpMode=RCIM_Cubic, leaveTangent=21.0298, tangentMode=RCTM_User, time=0, value=0.025}
           {arriveTangent=-2.9757, interpMode=RCIM_Cubic, leaveTangent=-2.9757, tangentMode=RCTM_User, time=0.3, value=1}
           {interpMode=RCIM_Cubic, tangentMode=RCTM_User, time=1, value=0}
          ]}}}
       CurveIndex=$Particles.PartitionParticles.Normalized Index In Partition
      )
      Spawn Group Mask=&SelectIntFromArray(Direct Array Index=$Particles.PartitionParticles.Partition, Int Selection Array=DataInterface() {propertyValues {IntData [0,0,0,0,1,1]}})
     )
     B=&Multiply_Float(
      A=&SelectFloatFromArray(Direct Array Index=$Particles.PartitionParticles.Partition, Float Selection Array=DataInterface() {propertyValues {FloatData [5,5,2,2,3,3]}})
      B=$Particles.Wave
     )
    )
    Set Material Dynamic=false
    SkeletalMesh:NiagaraDataInterfaceSkeletalMesh=$System.BoneFlowSkeleMesh1
    Use Sockets Instead=true
   )
   &ScaleColor(
    Scale Alpha=&MaskFloatBySpawnGroup(
     Masked Float=$Particles.QualityLevelOpacity
     Passthrough Float=&FloatFromCurve(
      FloatCurve=DataInterface() {propertyValues {Curve {keys [{interpMode=RCIM_Cubic, tangentMode=RCTM_User, time=0, value=1}, {interpMode=RCIM_Cubic, tangentMode=RCTM_User, time=0.8, value=0}]}}}
      CurveIndex=$Particles.PartitionParticles.Normalized Index In Partition
      Scale Curve=$Particles.QualityLevelOpacity
     )
     Spawn Group Mask=&SelectIntFromArray(Direct Array Index=$Particles.PartitionParticles.Partition, Int Selection Array=DataInterface() {propertyValues {IntData [0,0,1,1,1]}})
    )
    Scale RGB=&Lerp_Vector(
     A=<1 1 1>
     B=&MakeVectorFromLinearColorRGB(Color=&ColorFromCurve(
       ColorCurve=DataInterface() {propertyValues {
         AlphaCurve {keys [
           {interpMode=RCIM_Cubic, leaveTangentWeight=0.3333, time=0, value=1}
           {arriveTangentWeight=0.3333, interpMode=RCIM_Cubic, time=0.5009, value=1}
          ]}
         BlueCurve {keys [
           {interpMode=RCIM_Cubic, tangentMode=RCTM_User, time=0, value=0}
           {interpMode=RCIM_Cubic, leaveTangentWeight=0.1194, tangentMode=RCTM_User, time=0.2136, value=0}
           {
            arriveTangent=1.7758
            arriveTangentWeight=0.1474
            interpMode=RCIM_Cubic
            leaveTangent=1.7758
            leaveTangentWeight=0.0666
            time=0.4622
            value=0.7106
           }
           {arriveTangent=-0.2449, arriveTangentWeight=0.0666, interpMode=RCIM_Cubic, leaveTangent=-0.2449, time=0.7387, value=0.9324}
           {interpMode=RCIM_Cubic, tangentMode=RCTM_User, time=0.9608, value=0.5885}
          ]}
         GreenCurve {keys [
           {interpMode=RCIM_Cubic, tangentMode=RCTM_User, time=0, value=0.5781}
           {arriveTangent=-1.2509, interpMode=RCIM_Cubic, leaveTangent=-1.2509, leaveTangentWeight=0.1194, time=0.2136, value=0.8802}
           {
            arriveTangent=0.1488
            arriveTangentWeight=0.1474
            interpMode=RCIM_Cubic
            leaveTangent=0.1488
            leaveTangentWeight=0.0666
            time=0.4622
            value=0
           }
           {arriveTangent=0.5821, arriveTangentWeight=0.0666, interpMode=RCIM_Cubic, leaveTangent=0.5821, time=0.7387, value=0.9583}
           {interpMode=RCIM_Cubic, tangentMode=RCTM_User, time=0.9608, value=0.2903}
          ]}
         RedCurve {keys [
           {interpMode=RCIM_Cubic, tangentMode=RCTM_User, time=0, value=0.0241}
           {arriveTangent=2.1116, interpMode=RCIM_Cubic, leaveTangent=2.1116, leaveTangentWeight=0.1194, time=0.2136, value=0.7506}
           {arriveTangentWeight=0.2294, interpMode=RCIM_Cubic, leaveTangentWeight=0.1036, tangentMode=RCTM_User, time=0.4622, value=1}
           {arriveTangent=-2.0054, arriveTangentWeight=0.0666, interpMode=RCIM_Cubic, leaveTangent=-2.0054, time=0.7387, value=0}
           {interpMode=RCIM_Cubic, tangentMode=RCTM_User, time=0.9608, value=0}
          ]}
        }}
       CurveIndex=&Multiply_Float(A=$Particles.PartitionParticles.Normalized Index In Partition, B=$System.Timeline.Time)
      ))
     Alpha=$System.AmplitudeValue
    )
    ScaleA=true
    ScaleRGB=true
   ) {module=ScaleColor001}
   &DynamicMaterialParameters(
    Index 0 Param 3=&MakeCustomFloatFromBool(
     Bool=$System.bIsReactive
     True Float=$System.ReactiveValue
     False Float=&FloatFromCurve(
      FloatCurve=DataInterface() {propertyValues {Curve {keys [
          {interpMode=RCIM_Cubic, leaveTangent=4.8905, tangentMode=RCTM_Break, time=0, value=0.15}
          <at 0.2 0.9>
          {interpMode=RCIM_Cubic, leaveTangent=-1.0585, tangentMode=RCTM_Break, time=0.3, value=0.9}
          {arriveTangent=-0.6609, tangentMode=RCTM_User, time=1, value=0.15}
         ]}}}
      CurveIndex=$System.NormalizedLoopAge
     )
    )
    Index 0 Param 4=&MakeCustomFloatFromBool(Bool=$System.bIsReactive, True Float=1, False Float=0)
    Module.Index0Param0Write=true
    Module.Index0Param1Write=true
    Module.Index0Param2Write=true
    Module.Index0Param3Write=true
   )
   /SetVariables(Particles.GlowtubeColor:LinearColor=&MakeLinearColorFromVectorAndFloat(
     "Vector (RGB)"=<0.2 0.2 0.2>
     "Float (Alpha)"=&FloatFromCurve(
      FloatCurve=DataInterface() {propertyValues {Curve {keys [
          {arriveTangent=21.1196, interpMode=RCIM_Cubic, leaveTangent=21.1196, tangentMode=RCTM_User, time=0, value=0}
          {interpMode=RCIM_Cubic, tangentMode=RCTM_User, time=0.05, value=1}
          <at 0.5 0 RCIM_Cubic>
          <at 0.95 1 RCIM_Cubic>
          {arriveTangent=-20.6713, interpMode=RCIM_Cubic, leaveTangent=-20.6713, tangentMode=RCTM_User, time=1, value=0}
         ]}}}
      CurveIndex=$Particles.PartitionParticles.Normalized Index In Partition
      Scale Curve=&MaskFloatBySpawnGroup(
       Masked Float=&Multiply_Float(A=0.25, B=$Particles.QualityLevelOpacity)
       Passthrough Float=&FloatFromCurve(
        FloatCurve=DataInterface() {propertyValues {Curve {keys [{tangentMode=RCTM_User, time=0, value=1}, {tangentMode=RCTM_User, time=0.7, value=0}]}}}
        CurveIndex=$Particles.PartitionParticles.Normalized Index In Partition
        Scale Curve=&Multiply_Float(A=0.25, B=$Particles.QualityLevelOpacity)
       )
       Spawn Group Mask=&SelectIntFromArray(Direct Array Index=$Particles.PartitionParticles.Partition, Int Selection Array=DataInterface() {propertyValues {IntData [0,0,0,0,1,1]}})
      )
     )
    )) {isSetParameters=true, module=SetVariables_#10}
   &ScaleColor(
    Scale Mode=::NewEnumerator1
    Scale RGBA=&MakeVector4_FromLinearColor(LinearColor=&FNTimeOfDayColor(Night Color=<rgba 0.4683 0.4038 0.9 1>, Morning Color=<rgba 1 0.6536 0.6237 1>, Day Color=<rgba 1 1 1 1>, Evening Color=<rgba 1 1 1 1>))
    ScaleRGBA=true
   )
  ]
  rendererClasses [&NiagaraRibbonRendererProperties]
  renderers [
   {rendererClass=&NiagaraRibbonRendererProperties}
   {rendererClass=&NiagaraRibbonRendererProperties, rendererIndex=1}
  ]
 }
 ArmSplines001 {
  bEnabled=false
  emitterUpdateScript [
   &EmitterState(Life Cycle Mode=::NewEnumerator1, Loop Behavior=::NewEnumerator1, Loop Duration Mode=::NewEnumerator1, UseLoopDelay=false)
   &SpawnBurst_Instantaneous(Spawn Count=36, Use Spawn Probability=false, UseLoopCountLimit=false)
  ]
  particleSpawnScript [
   &PartitionParticles(
    Normalized Index In Partition Includes One=true
    Number Of Partitions=6
    Particles Per Partition=6
    Partition Mode:ENiagara_PartitionMode=::NewEnumerator1
    Write Partition to Ribbon ID=::NewEnumerator0
   )
   &InitializeParticle(
    Color=<rgba 1 1 1 0.33>
    Lifetime=5
    Ribbon Width=7
    Ribbon Width Mode=::NewEnumerator1
    Sprite Size Mode=::NewEnumerator3
    Uniform Sprite Size=2
    UsePositionOffset=false
    Write Material Random=true
   )
  ]
  particleUpdateScript [
   /SetVariables(Particles.QualityLevelOpacity:NiagaraFloat=&Lerp_Float(
     A=1
     B=&FloatFromCurve(
      FloatCurve=DataInterface() {propertyValues {Curve {keys [
          {interpMode=RCIM_Cubic, tangentMode=RCTM_User, time=0.05, value=0}
          {arriveTangent=1.4286, time=0.3, value=1}
          {interpMode=RCIM_Cubic, tangentMode=RCTM_User, time=0.75, value=1}
          {tangentMode=RCTM_User, time=1, value=0}
         ]}}}
      CurveIndex=$Particles.PartitionParticles.Normalized Index In Partition
     )
     Alpha=&MakeFloatFromInt(Int=&SelectIntFromArray(Direct Array Index=$Engine.QualityLevel, Int Selection Array=DataInterface() {propertyValues {IntData [1,1,0,0,0]}}))
    )) {isSetParameters=true, module=SetVariables_#7}
   /SetVariables(Particles.Wave:NiagaraFloat=&Waveform(
     Waveform Count=::NewEnumerator1
     Global Input Position=&FloatFromCurve(
      FloatCurve=DataInterface() {propertyValues {Curve {keys [{tangentMode=RCTM_User, time=0, value=0}, {tangentMode=RCTM_User, time=1, value=1}]}}}
      CurveIndex=$Particles.PartitionParticles.Normalized Index In Partition
      Scale Curve=&Multiply_Float(
       A=1
       B=&FloatFromCurve(
        FloatCurve=DataInterface() {propertyValues {Curve {keys [
            {interpMode=RCIM_Cubic, tangentMode=RCTM_User, time=0.1, value=0}
            {arriveTangent=1.1111, time=0.2, value=1}
            {interpMode=RCIM_Cubic, tangentMode=RCTM_User, time=0.8, value=1}
            {tangentMode=RCTM_User, time=0.9, value=0}
           ]}}}
        CurveIndex=$Particles.PartitionParticles.Normalized Index In Partition
       )
      )
     )
     "Waveform [1]"=::NewEnumerator0
     "[1] Phase Scale"=0.25
     "[1] Frequency"=1.5
     "[1] Amplitude Min/Max"=<0 1>
     "[1] Scale"=0.5
     "[1] Bias"=1
     "[2] Average"=false
     "[2] Input Position"=0
     "[2] Phase"=$Emitter.Age
     "[2] Phase Scale"=-0.5
     "[2] Compound Frequency"=2
     "[2] Frequency"=2
     "[2] Amplitude Min/Max"=<0 1>
     "[2] Scale"=&Lerp_Float(A=0, B=1, Alpha=$System.AmplitudeValue)
     "[2] Bias"=0
     "[2] Exponent"=1
     "[2] Pulse Duration"=0.1
     "[2] Clip Triangle Min/Max"=<0 0>
     "[2] Random Seed"=0
    )) {isSetParameters=true, module=SetVariables_#8}
   &ScaleRibbonWidth(Ribbon Width Scale=&FloatFromCurve(
     FloatCurve=DataInterface() {propertyValues {Curve {keys [
         {arriveTangent=6.0891, interpMode=RCIM_Cubic, leaveTangent=6.0891, tangentMode=RCTM_User, time=0, value=1}
         {arriveTangent=1.25, time=0.1, value=1}
         {interpMode=RCIM_Cubic, tangentMode=RCTM_User, time=0.9, value=1}
         {arriveTangent=-4.9474, tangentMode=RCTM_User, time=1, value=1}
        ]}}}
     CurveIndex=$Particles.Initial.PartitionParticles.Normalized Index In Partition
     Scale Curve=&Lerp_Float(A=1, B=1.25, Alpha=$System.AmplitudeValue)
    ))
   /SetVariables(Particles.GlowRibbonWidth:NiagaraFloat=&Multiply_Float(
     A=$Particles.Wave
     B=&Lerp_Float(
      A=&FloatFromCurve(
       FloatCurve=DataInterface() {propertyValues {Curve {keys [
           {interpMode=RCIM_Cubic, tangentMode=RCTM_User, time=0, value=1}
           {arriveTangent=-2.0527, tangentMode=RCTM_User, time=0.2, value=0.6}
           {arriveTangent=1.9371, interpMode=RCIM_Cubic, leaveTangent=1.9371, tangentMode=RCTM_User, time=0.8, value=0.6}
           {tangentMode=RCTM_User, time=1, value=1}
          ]}}}
       CurveIndex=$Particles.PartitionParticles.Normalized Index In Partition
       Scale Curve=4
      )
      B=4
      Alpha=$System.AmplitudeValue
     )
    )) {isSetParameters=true, module=SetVariables_#9}
   &ParticleState(
    Kill Particles When Lifetime Has Elapsed=false
    Let Infinitely Lived Particles Die When Emitter Deactivates=true
    Loop Particles Lifetime=true
   )
   &FN_BoneFlow(
    Angle:NiagaraFloat=&Add_Float(
     A=&SelectFloatFromArray(
      Direct Array Index=$Particles.PartitionParticles.Partition
      Float Selection Array=DataInterface() {propertyValues {FloatData [0.25,0.25,0.25,0.25,0.225,0.13]}}
     )
     B=&FloatFromCurve(
      FloatCurve=DataInterface() {propertyValues {Curve {keys [{interpMode=RCIM_Cubic, tangentMode=RCTM_User, time=0, value=0}, {interpMode=RCIM_Cubic, tangentMode=RCTM_User, time=0.8, value=1}]}}}
      CurveIndex=$Particles.PartitionParticles.Normalized Index In Partition
      Scale Curve=&SelectFloatFromArray(
       Direct Array Index=$Particles.PartitionParticles.Partition
       Float Selection Array=DataInterface() {propertyValues {FloatData [0,0,0,0,-0.3,0.4]}}
      )
     )
    )
    BoneID:NiagaraInt32=$Particles.Initial.PartitionParticles.Partition
    Enable Position Offsets=true
    Height:NiagaraFloat=$Particles.PartitionParticles.Normalized Index In Partition
    Offset End:Vector3f=&SelectVectorFromArray(
     Direct Array Index=$Particles.PartitionParticles.Partition
     Vector Selection Array=DataInterface() {propertyValues {FloatData [<0 0 2>, <0 0 -2>, <0 0 0>, <0 0 0>, <0 0 0>, <0 0 0>]}}
    )
    Offset Mid:Vector3f=&SelectVectorFromArray(
     Direct Array Index=$Particles.PartitionParticles.Partition
     Vector Selection Array=DataInterface() {propertyValues {FloatData [<0 0 0>, <0 0 0>, <0 0 0>, <0 0 0>, <0 0 0>, <0 0 0>]}}
    )
    Offset Start:Vector3f=&SelectVectorFromArray(
     Direct Array Index=$Particles.PartitionParticles.Partition
     Vector Selection Array=DataInterface() {propertyValues {FloatData [<0 0 0>, <0 0 0>, <0 0 0>, <0 0 0>, <-2.5 0 -2>, <-2.5 0 -2>]}}
    )
    Radius:NiagaraFloat=&Multiply_Float(
     A=&MaskFloatBySpawnGroup(
      Masked Float=&FloatFromCurve(
       FloatCurve=DataInterface() {propertyValues {Curve {keys [
           {arriveTangent=13.5458, interpMode=RCIM_Cubic, leaveTangent=13.5458, tangentMode=RCTM_User, time=0, value=0.5}
           {interpMode=RCIM_Cubic, tangentMode=RCTM_User, time=0.5, value=1}
           {arriveTangent=-13.0583, tangentMode=RCTM_User, time=1, value=0.5}
          ]}}}
       CurveIndex=$Particles.PartitionParticles.Normalized Index In Partition
      )
      Passthrough Float=&FloatFromCurve(
       FloatCurve=DataInterface() {propertyValues {Curve {keys [
           {arriveTangent=21.0298, interpMode=RCIM_Cubic, leaveTangent=21.0298, tangentMode=RCTM_User, time=0, value=0.025}
           {arriveTangent=-2.9757, interpMode=RCIM_Cubic, leaveTangent=-2.9757, tangentMode=RCTM_User, time=0.3, value=1}
           {interpMode=RCIM_Cubic, tangentMode=RCTM_User, time=1, value=0}
          ]}}}
       CurveIndex=$Particles.PartitionParticles.Normalized Index In Partition
      )
      Spawn Group Mask=&SelectIntFromArray(Direct Array Index=$Particles.PartitionParticles.Partition, Int Selection Array=DataInterface() {propertyValues {IntData [0,0,0,0,1,1]}})
     )
     B=&Multiply_Float(
      A=&SelectFloatFromArray(Direct Array Index=$Particles.PartitionParticles.Partition, Float Selection Array=DataInterface() {propertyValues {FloatData [5,5,2,2,3,3]}})
      B=$Particles.Wave
     )
    )
    Set Material Dynamic=false
    SkeletalMesh:NiagaraDataInterfaceSkeletalMesh=$System.BoneFlowSkeleMesh1
    Use Sockets Instead=true
    Write Position Attribute=false
   )
   /SetVariables(Particles.GoalPos:NiagaraPosition=$Particles.FN_BoneFlow.Position) {isSetParameters=true}
   /SetVariables(Particles.DistanceFromGoal:NiagaraFloat=&ClampFloat(
     Float=&Divide_Float(
      A=&VectorLength(Vector=&Subtract_Position(A=$Particles.Position, B=$Particles.GoalPos))
      B=&Multiply_Float(
       A=&FloatFromCurve(
        FloatCurve=DataInterface() {propertyValues {Curve {keys [
            {interpMode=RCIM_Cubic, tangentMode=RCTM_User, time=0, value=0}
            {interpMode=RCIM_Cubic, tangentMode=RCTM_User, time=0.5, value=1}
            {tangentMode=RCTM_User, time=1, value=0}
           ]}}}
        CurveIndex=$Particles.PartitionParticles.Normalized Index In Partition
       )
       B=15
      )
     )
     Min=0.01
    )) {isSetParameters=true}
   &InterpolateOverTime(Target Float Value:NiagaraFloat=$Particles.DistanceFromGoal) {module=InterpolateOverTime001}
   /SetVariables(Particles.Position=&Lerp_Vector(
     A=Linked() {linkedVariable {name=Particles.GoalPos, type=NiagaraPosition}}
     B=&Lerp_Vector(
      A=Linked() {linkedVariable {name=Particles.Position, type=NiagaraPosition}}
      B=Linked() {linkedVariable {name=Particles.GoalPos, type=NiagaraPosition}}
      Alpha=$Particles.InterpolateOverTime001.InterpolateOverTime_Float.Moving Average
     )
     Alpha=&FloatFromCurve(
      FloatCurve=DataInterface() {propertyValues {Curve {keys [
          {interpMode=RCIM_Cubic, tangentMode=RCTM_User, time=0, value=0}
          {interpMode=RCIM_Cubic, tangentMode=RCTM_User, time=0.5, value=1}
          {tangentMode=RCTM_User, time=1, value=0}
         ]}}}
      CurveIndex=/Niagara/DynamicInputs/Math/Min_Float(
       A:NiagaraFloat=$Particles.PartitionParticles.Normalized Index In Partition
       B:NiagaraFloat=&OneMinusFloat(Float=$Particles.DistanceFromGoal)
      )
      Scale Curve=0.5
     )
    )) {isSetParameters=true}
   &ScaleColor(
    Scale Alpha=&FloatFromCurve(
     FloatCurve=DataInterface() {propertyValues {Curve {keys [
         {arriveTangent=0.0953, interpMode=RCIM_Cubic, leaveTangent=0.0953, tangentMode=RCTM_User, time=0.1, value=0}
         {arriveTangent=0.0485, interpMode=RCIM_Cubic, leaveTangent=0.0485, tangentMode=RCTM_User, time=0.4, value=1}
         {arriveTangent=-0.3611, interpMode=RCIM_Cubic, leaveTangent=-0.3611, tangentMode=RCTM_User, time=0.6, value=1}
         {interpMode=RCIM_Cubic, tangentMode=RCTM_User, time=0.9, value=0}
        ]}}}
     CurveIndex=$Particles.PartitionParticles.Normalized Index In Partition
     Scale Curve=&Multiply_Float(A=1, B=$Particles.QualityLevelOpacity)
    )
    Scale RGB=&Lerp_Vector(
     A=<1 1 1>
     B=&MakeVectorFromLinearColorRGB(Color=&ColorFromCurve(
       ColorCurve=DataInterface() {propertyValues {
         AlphaCurve {keys [
           {interpMode=RCIM_Cubic, leaveTangentWeight=0.3333, time=0, value=1}
           {arriveTangentWeight=0.3333, interpMode=RCIM_Cubic, time=0.5009, value=1}
          ]}
         BlueCurve {keys [
           {interpMode=RCIM_Cubic, tangentMode=RCTM_User, time=0, value=0}
           {interpMode=RCIM_Cubic, leaveTangentWeight=0.1194, tangentMode=RCTM_User, time=0.2136, value=0}
           {
            arriveTangent=1.7758
            arriveTangentWeight=0.1474
            interpMode=RCIM_Cubic
            leaveTangent=1.7758
            leaveTangentWeight=0.0666
            time=0.4622
            value=0.7106
           }
           {arriveTangent=-0.2449, arriveTangentWeight=0.0666, interpMode=RCIM_Cubic, leaveTangent=-0.2449, time=0.7387, value=0.9324}
           {interpMode=RCIM_Cubic, tangentMode=RCTM_User, time=0.9608, value=0.5885}
          ]}
         GreenCurve {keys [
           {interpMode=RCIM_Cubic, tangentMode=RCTM_User, time=0, value=0.5781}
           {arriveTangent=-1.2509, interpMode=RCIM_Cubic, leaveTangent=-1.2509, leaveTangentWeight=0.1194, time=0.2136, value=0.8802}
           {
            arriveTangent=0.1488
            arriveTangentWeight=0.1474
            interpMode=RCIM_Cubic
            leaveTangent=0.1488
            leaveTangentWeight=0.0666
            time=0.4622
            value=0
           }
           {arriveTangent=0.5821, arriveTangentWeight=0.0666, interpMode=RCIM_Cubic, leaveTangent=0.5821, time=0.7387, value=0.9583}
           {interpMode=RCIM_Cubic, tangentMode=RCTM_User, time=0.9608, value=0.2903}
          ]}
         RedCurve {keys [
           {interpMode=RCIM_Cubic, tangentMode=RCTM_User, time=0, value=0.0241}
           {arriveTangent=2.1116, interpMode=RCIM_Cubic, leaveTangent=2.1116, leaveTangentWeight=0.1194, time=0.2136, value=0.7506}
           {arriveTangentWeight=0.2294, interpMode=RCIM_Cubic, leaveTangentWeight=0.1036, tangentMode=RCTM_User, time=0.4622, value=1}
           {arriveTangent=-2.0054, arriveTangentWeight=0.0666, interpMode=RCIM_Cubic, leaveTangent=-2.0054, time=0.7387, value=0}
           {interpMode=RCIM_Cubic, tangentMode=RCTM_User, time=0.9608, value=0}
          ]}
        }}
       CurveIndex=&Multiply_Float(A=$Particles.PartitionParticles.Normalized Index In Partition, B=$System.Timeline.Time)
      ))
     Alpha=$System.AmplitudeValue
    )
    ScaleA=true
    ScaleRGB=true
   ) {module=ScaleColor001}
   &DynamicMaterialParameters(
    Index 0 Param 3=&MakeCustomFloatFromBool(
     Bool=$System.bIsReactive
     True Float=$System.ReactiveValue
     False Float=&FloatFromCurve(
      FloatCurve=DataInterface() {propertyValues {Curve {keys [
          {interpMode=RCIM_Cubic, leaveTangent=4.8905, tangentMode=RCTM_Break, time=0, value=0.15}
          <at 0.2 0.9>
          {interpMode=RCIM_Cubic, leaveTangent=-1.0585, tangentMode=RCTM_Break, time=0.3, value=0.9}
          {arriveTangent=-0.6609, tangentMode=RCTM_User, time=1, value=0.15}
         ]}}}
      CurveIndex=$System.NormalizedLoopAge
     )
    )
    Index 0 Param 4=&MakeCustomFloatFromBool(Bool=$System.bIsReactive, True Float=1, False Float=0)
    Module.Index0Param0Write=true
    Module.Index0Param1Write=true
    Module.Index0Param2Write=true
    Module.Index0Param3Write=true
   )
   /SetVariables(Particles.GlowtubeColor:LinearColor=&MakeLinearColorFromVectorAndFloat(
     "Vector (RGB)"=<0.2 0.2 0.2>
     "Float (Alpha)"=&FloatFromCurve(
      FloatCurve=DataInterface() {propertyValues {Curve {keys [
          {arriveTangent=21.1196, interpMode=RCIM_Cubic, leaveTangent=21.1196, tangentMode=RCTM_User, time=0, value=0}
          {interpMode=RCIM_Cubic, tangentMode=RCTM_User, time=0.05, value=1}
          <at 0.5 0 RCIM_Cubic>
          <at 0.95 1 RCIM_Cubic>
          {arriveTangent=-20.6713, interpMode=RCIM_Cubic, leaveTangent=-20.6713, tangentMode=RCTM_User, time=1, value=0}
         ]}}}
      CurveIndex=$Particles.PartitionParticles.Normalized Index In Partition
      Scale Curve=&MaskFloatBySpawnGroup(
       Masked Float=&Multiply_Float(A=0.25, B=$Particles.QualityLevelOpacity)
       Passthrough Float=&FloatFromCurve(
        FloatCurve=DataInterface() {propertyValues {Curve {keys [{tangentMode=RCTM_User, time=0, value=1}, {tangentMode=RCTM_User, time=0.7, value=0}]}}}
        CurveIndex=$Particles.PartitionParticles.Normalized Index In Partition
        Scale Curve=&Multiply_Float(A=0.25, B=$Particles.QualityLevelOpacity)
       )
       Spawn Group Mask=&SelectIntFromArray(Direct Array Index=$Particles.PartitionParticles.Partition, Int Selection Array=DataInterface() {propertyValues {IntData [0,0,0,0,1,1]}})
      )
     )
    )) {isSetParameters=true, module=SetVariables_#10}
   &ScaleColor(
    Scale Mode=::NewEnumerator1
    Scale RGBA=&MakeVector4_FromLinearColor(LinearColor=&FNTimeOfDayColor(Night Color=<rgba 0.4683 0.4038 0.9 1>, Morning Color=<rgba 1 0.6536 0.6237 1>, Day Color=<rgba 1 1 1 1>, Evening Color=<rgba 1 1 1 1>))
    ScaleRGBA=true
   )
  ]
  rendererClasses [&NiagaraRibbonRendererProperties]
  renderers [
   {rendererClass=&NiagaraRibbonRendererProperties}
   {rendererClass=&NiagaraRibbonRendererProperties, rendererIndex=1}
  ]
 }
 BodyLights {
  emitterUpdateScript [
   &EmitterState(Life Cycle Mode=::NewEnumerator1, Loop Behavior=::NewEnumerator1, Loop Duration Mode=::NewEnumerator1, UseLoopDelay=false)
   &SpawnBurst_Instantaneous(Spawn Count=4, Use Spawn Probability=false, UseLoopCountLimit=false)
  ]
  particleSpawnScript [
   &PartitionParticles(Number Of Partitions=4)
   &InitializeParticle(
    Color Mode=::NewEnumerator0
    Lifetime=5
    Sprite Size Mode=::NewEnumerator3
    Uniform Sprite Size=50
    UsePositionOffset=false
    Write Material Random=true
   )
  ]
  particleUpdateScript [
   &ParticleState(
    Kill Particles When Lifetime Has Elapsed=false
    Let Infinitely Lived Particles Die When Emitter Deactivates=true
    Loop Particles Lifetime=true
   )
   &SkeletalMeshLocation(
    Bone / Socket Index=&SelectIntFromArray(Direct Array Index=$Particles.PartitionParticles.Partition, Int Selection Array=DataInterface() {propertyValues {IntData [1,2,3,4]}})
    Bone Sampling Mode=::NewEnumerator4
    MaskSpawnGroup=false
    Sampled Position Offset=&SelectVectorFromArray(
     Direct Array Index=$Particles.PartitionParticles.Partition
     Vector Selection Array=DataInterface() {propertyValues {FloatData [<0 5 10>, <0 -5 -10>, <10 5 -5>, <-10 -5 5>]}}
    )
    Skeletal Mesh=$System.BoneFlowSkeleMesh1
   )
   &ScaleColor(
    Scale Alpha=$System.TOD
    Scale RGB=&MakeVectorFromLinearColorRGB(Color=&ColorFromCurve(
      ColorCurve=DataInterface() {propertyValues {
        AlphaCurve {keys [
          {interpMode=RCIM_Cubic, leaveTangentWeight=0.3333, time=0, value=1}
          {arriveTangentWeight=0.3333, interpMode=RCIM_Cubic, time=0.5009, value=1}
         ]}
        BlueCurve {keys [
          {interpMode=RCIM_Cubic, tangentMode=RCTM_User, time=0, value=0.3137}
          {arriveTangentWeight=0.0666, interpMode=RCIM_Cubic, leaveTangent=0.2746, time=0.4951, value=0.7454}
          <at 1 0.3137>
         ]}
        GreenCurve {keys [
          {interpMode=RCIM_Cubic, tangentMode=RCTM_User, time=0, value=0.6431}
          {arriveTangentWeight=0.0666, interpMode=RCIM_Cubic, leaveTangent=-0.313, time=0.4951, value=0.7605}
          <at 1 0.6431>
         ]}
        RedCurve {keys [
          {interpMode=RCIM_Cubic, tangentMode=RCTM_User, time=0, value=0.3294}
          {arriveTangentWeight=0.0666, interpMode=RCIM_Cubic, leaveTangent=-0.2465, time=0.4951, value=0.2016}
          <at 1 0.3294>
         ]}
       }}
      CurveIndex=&Lerp_Float(A=$System.NormalizedLoopAge, B=$System.AmplitudeValue, Alpha=$System.Timeline.Time)
     ))
    ScaleA=true
    ScaleRGB=true
   )
   &FN_Light_Attributes(
    Color=$Particles.Color
    Create Color Attribute=true
    Create Exponent Attribute=true
    Create Position Attribute=true
    Create Radius Attribute=true
    Exponent=2
    Position:NiagaraPosition=$Particles.Position
    Radius=40
    Scale Color=3
   )
  ]
  rendererClasses [&NiagaraLightRendererProperties]
  renderers [{rendererClass=&NiagaraLightRendererProperties, rendererIndex=0}]
 }
 Bone_Flow_Hip_Left {
  emitterSpawnScript [
   /SetVariables(Emitter.RandomSeed:NiagaraInt32=/Niagara/DynamicInputs/Random/RandomInt(Randomness Mode:ENiagaraRandomnessMode=::NewEnumerator0, Random Seed:NiagaraInt32=0, Fixed Random Seed:NiagaraBool=false)) {isSetParameters=true, module=SetVariables_#11}
   /SetVariables(Emitter.ParticlesPerGroup=6) {isSetParameters=true, module=SetVariables_#12}
  ]
  emitterUpdateScript [
   /SetVariables(Emitter.NumberOfGroups=1) {isSetParameters=true, module=SetVariables_#13}
   &EmitterState(Life Cycle Mode=::NewEnumerator1, Loop Behavior=::NewEnumerator1, Loop Duration Mode=::NewEnumerator1, UseLoopDelay=false)
   &SpawnBurst_Instantaneous(
    Spawn Count=&Multiply_Int(A=$Emitter.ParticlesPerGroup, B=$Emitter.NumberOfGroups)
    Use Spawn Probability=false
   )
  ]
  particleSpawnScript [
   /SetVariables(Particles.LoopCount:NiagaraInt32=$Emitter.LoopCount) {isSetParameters=true, module=SetVariables_#14}
   &FN_SetSpawnGroup(
    Inclusivity=true
    Normalize Spawn Group=true
    Particles Per Group:NiagaraInt32=$Emitter.ParticlesPerGroup
    Pattern=::NewEnumerator0
    Set Ribbon ID and Ribbon Link Order=true
   )
   /SetVariables(Particles.QualityLevelOpacity:NiagaraFloat=&Lerp_Float(
     A=1
     B=&FloatFromCurve(
      FloatCurve=DataInterface() {propertyValues {Curve {keys [
          {interpMode=RCIM_Cubic, tangentMode=RCTM_User, time=0.1, value=0}
          {arriveTangent=1.4286, time=0.4, value=1}
          {interpMode=RCIM_Cubic, tangentMode=RCTM_User, time=0.75, value=1}
          {tangentMode=RCTM_User, time=1, value=0}
         ]}}}
      CurveIndex=$Particles.Normalized Spawn Group
     )
     Alpha=&MakeFloatFromInt(Int=&SelectIntFromArray(Direct Array Index=$Engine.QualityLevel, Int Selection Array=DataInterface() {propertyValues {IntData [1,1,0,0,0]}}))
    )) {isSetParameters=true}
   &InitializeRibbon(
    Color=<rgba 1 1 1 0.33>
    Lifetime=5
    Module.Write Color=true
    Module.Write Lifetime=true
    Module.Write Mass=true
    Module.Write Material Random=true
    Module.Write Position=false
    Module.Write RibbonTwist=false
    Module.Write RibbonWidth=true
    Ribbon Width=8
   )
   &FN_FlipRibbonV() {enabled=false}
  ]
  particleUpdateScript [
   &ParticleState(Kill Particles When Lifetime Has Elapsed=false, Let Infinitely Lived Particles Die When Emitter Deactivates=true)
   /SetVariables(Particles.Wave:NiagaraFloat=&Waveform(
     Global Input Position=$Particles.Normalized Spawn Group
     Global Amplitude Scale=0.7
     "Waveform [1]"=::NewEnumerator0
     "[1] Phase Scale"=0.5
     "[1] Frequency"=1.5
     "[1] Amplitude Min/Max"=<0 1>
     "[1] Scale"=&Lerp_Float(A=0, B=1, Alpha=$System.AmplitudeValue)
    )) {isSetParameters=true}
   &ScaleRibbonWidth(Ribbon Width Scale=&FloatFromCurve(
     FloatCurve=DataInterface() {propertyValues {Curve {keys [
         {arriveTangent=6.0891, interpMode=RCIM_Cubic, leaveTangent=6.0891, tangentMode=RCTM_User, time=0, value=1}
         {arriveTangent=1.25, time=0.1, value=1}
         {interpMode=RCIM_Cubic, tangentMode=RCTM_User, time=0.9, value=1}
         {arriveTangent=-4.9474, tangentMode=RCTM_User, time=1, value=1}
        ]}}}
     CurveIndex=$Particles.Normalized Spawn Group
     Scale Curve=&Lerp_Float(A=1, B=1.25, Alpha=$System.AmplitudeValue)
    ))
   &FN_BoneFlow(
    Angle:NiagaraFloat=&FloatFromCurve(
     FloatCurve=DataInterface() {propertyValues {Curve {keys [<at -0.3 0.25 RCIM_Cubic>, {arriveTangent=0.5769, interpMode=RCIM_Cubic, leaveTangent=0.5769, time=0.7, value=1.875}, <at 1 1 RCIM_Cubic>]}}}
     CurveIndex=$Particles.Normalized Spawn Group
     Scale Curve=0.256
    )
    BoneID:NiagaraInt32=&Add_Integer(A=$Particles.SpawnGroup, B=5)
    Height:NiagaraFloat=$Particles.Normalized Spawn Group
    Radius:NiagaraFloat=&FloatFromCurve(
     FloatCurve=DataInterface() {propertyValues {Curve {keys [
         {arriveTangent=59.485, interpMode=RCIM_Cubic, leaveTangent=59.485, tangentMode=RCTM_User, time=0, value=0}
         {arriveTangent=7, interpMode=RCIM_Cubic, leaveTangent=7, time=0.2, value=5}
         {arriveTangent=-6.25, interpMode=RCIM_Cubic, leaveTangent=-6.25, time=0.5, value=3.5}
         {arriveTangent=-14.1598, interpMode=RCIM_Cubic, leaveTangent=-14.1598, tangentMode=RCTM_User, time=1, value=0}
        ]}}}
     CurveIndex=$Particles.Normalized Spawn Group
     Scale Curve=&Add_Float(A=1, B=$Particles.Wave)
    )
    SkeletalMesh:NiagaraDataInterfaceSkeletalMesh=$System.BoneFlowSkeleMesh1
    Use Sockets Instead=true
    Write Position Attribute=true
   )
   /SetVariables(Particles.RibbonFacing=$Output.FN_BoneFlow.RotatedBiNormal) {enabled=false, isSetParameters=true, module=SetVariables_#15}
   &FN_RibbonMaterial(
    Color=<rgba 0 1 1 9>
    Opacity Pan U Offset:NiagaraFloat=/Game/Effects/Niagara/DynamicInputs/Random/FN_SeededRandomRangeFloat(
     Minimum:NiagaraFloat=0
     Maximum:NiagaraFloat=2
     Random Seed:NiagaraInt32=&Add_Integer(A=$Particles.LoopCount, B=$Particles.SpawnGroup)
    )
    Opacity Pan U Speed:NiagaraFloat=&FN_SeededRandomRangeFloat(
     Minimum:NiagaraFloat=0.1364
     Maximum:NiagaraFloat=0.5182
     Random Seed:NiagaraInt32=&Add_Integer(A=$Particles.LoopCount, B=$Particles.SpawnGroup)
    )
    Opacity Scale U:NiagaraFloat=&FN_SeededRandomRangeFloat(
     Minimum:NiagaraFloat=0.4273
     Maximum:NiagaraFloat=1.0032
     Random Seed:NiagaraInt32=&Add_Integer(A=$Particles.LoopCount, B=$Particles.SpawnGroup)
    )
    Scale Alpha:NiagaraFloat=&FloatFromCurve(
     FloatCurve=DataInterface() {propertyValues {Curve {keys [<at -0.007 0.0112>,<at 0.2367 0.9629>,<at 0.7537 0.9995>,<at 1 0>]}}}
     Scale Curve=&FloatFromCurve(
      FloatCurve=DataInterface() {propertyValues {Curve {keys [<at 0.0007 0.022>,<at 0.2494 1.0225>,<at 0.8089 0.9748>,<at 1 0>]}}}
      CurveIndex=$Particles.Normalized Spawn Group
     )
    )
    UV Basis:NiagaraFloat=$Particles.Normalized Spawn Group
   ) {enabled=false}
   &FN_Sprite_Attributes(Color=<rgba 1 0.597 0 1>, Create Color Attribute=true, Create SpriteSize Attribute=true, SpriteSize=<3 3>, Write to Module Namespace=true) {enabled=false}
   &ScaleColor(
    Scale Alpha=&FloatFromCurve(
     FloatCurve=DataInterface() {propertyValues {Curve {keys [
         {arriveTangent=4.2704, interpMode=RCIM_Cubic, leaveTangent=4.2704, tangentMode=RCTM_User, time=0, value=0}
         {interpMode=RCIM_Cubic, tangentMode=RCTM_User, time=0.2, value=1}
         {arriveTangent=-1.25, interpMode=RCIM_Cubic, leaveTangent=-1.25, time=0.5, value=1}
         {arriveTangent=-0.4071, tangentMode=RCTM_User, time=1, value=0}
        ]}}}
     CurveIndex=$Particles.Normalized Spawn Group
     Scale Curve=&SelectFloatFromArray(Direct Array Index=$Engine.QualityLevel, Float Selection Array=DataInterface() {propertyValues {FloatData [0.4,0.6,1,1,1]}})
    )
    Scale RGB=&Lerp_Vector(
     A=<1 1 1>
     B=&MakeVectorFromLinearColorRGB(Color=&ColorFromCurve(
       ColorCurve=DataInterface() {propertyValues {
         AlphaCurve {keys [
           {interpMode=RCIM_Cubic, leaveTangentWeight=0.3333, time=0, value=1}
           {arriveTangentWeight=0.3333, interpMode=RCIM_Cubic, time=0.5009, value=1}
          ]}
         BlueCurve {keys [
           {interpMode=RCIM_Cubic, tangentMode=RCTM_User, time=0, value=0}
           {interpMode=RCIM_Cubic, leaveTangentWeight=0.1194, tangentMode=RCTM_User, time=0.2136, value=0}
           {
            arriveTangent=1.7758
            arriveTangentWeight=0.1474
            interpMode=RCIM_Cubic
            leaveTangent=1.7758
            leaveTangentWeight=0.0666
            time=0.4622
            value=0.7106
           }
           {arriveTangent=-0.2449, arriveTangentWeight=0.0666, interpMode=RCIM_Cubic, leaveTangent=-0.2449, time=0.7387, value=0.9324}
           {interpMode=RCIM_Cubic, tangentMode=RCTM_User, time=0.9608, value=0.5885}
          ]}
         GreenCurve {keys [
           {interpMode=RCIM_Cubic, tangentMode=RCTM_User, time=0, value=0.5781}
           {arriveTangent=-1.2509, interpMode=RCIM_Cubic, leaveTangent=-1.2509, leaveTangentWeight=0.1194, time=0.2136, value=0.8802}
           {
            arriveTangent=0.1488
            arriveTangentWeight=0.1474
            interpMode=RCIM_Cubic
            leaveTangent=0.1488
            leaveTangentWeight=0.0666
            time=0.4622
            value=0
           }
           {arriveTangent=0.5821, arriveTangentWeight=0.0666, interpMode=RCIM_Cubic, leaveTangent=0.5821, time=0.7387, value=0.9583}
           {interpMode=RCIM_Cubic, tangentMode=RCTM_User, time=0.9608, value=0.2903}
          ]}
         RedCurve {keys [
           {interpMode=RCIM_Cubic, tangentMode=RCTM_User, time=0, value=0.0241}
           {arriveTangent=2.1116, interpMode=RCIM_Cubic, leaveTangent=2.1116, leaveTangentWeight=0.1194, time=0.2136, value=0.7506}
           {arriveTangentWeight=0.2294, interpMode=RCIM_Cubic, leaveTangentWeight=0.1036, tangentMode=RCTM_User, time=0.4622, value=1}
           {arriveTangent=-2.0054, arriveTangentWeight=0.0666, interpMode=RCIM_Cubic, leaveTangent=-2.0054, time=0.7387, value=0}
           {interpMode=RCIM_Cubic, tangentMode=RCTM_User, time=0.9608, value=0}
          ]}
        }}
       CurveIndex=&Multiply_Float(A=$Particles.Normalized Spawn Group, B=$System.Timeline.Time)
      ))
     Alpha=$System.AmplitudeValue
    )
    ScaleA=true
    ScaleRGB=true
   ) {module=ScaleColor001}
   /SetVariables(Particles.GlowtubeColor:LinearColor=&MakeLinearColorFromVectorAndFloat(
     "Vector (RGB)"=<0.2 0.2 0.2>
     "Float (Alpha)"=&FloatFromCurve(
      FloatCurve=DataInterface() {propertyValues {Curve {keys [
          {arriveTangent=21.1196, interpMode=RCIM_Cubic, leaveTangent=21.1196, tangentMode=RCTM_User, time=0, value=0}
          {interpMode=RCIM_Cubic, tangentMode=RCTM_User, time=0.05, value=1}
          <at 0.5 0 RCIM_Cubic>
          <at 0.95 1 RCIM_Cubic>
          {arriveTangent=-20.6713, interpMode=RCIM_Cubic, leaveTangent=-20.6713, tangentMode=RCTM_User, time=1, value=0}
         ]}}}
      CurveIndex=$Particles.Normalized Spawn Group
      Scale Curve=&MaskFloatBySpawnGroup(
       Masked Float=&Multiply_Float(A=0.25, B=$Particles.QualityLevelOpacity)
       Passthrough Float=&FloatFromCurve(
        FloatCurve=DataInterface() {propertyValues {Curve {keys [{tangentMode=RCTM_User, time=0, value=1}, {tangentMode=RCTM_User, time=0.7, value=0}]}}}
        CurveIndex=$Particles.Normalized Spawn Group
        Scale Curve=&Multiply_Float(A=0.25, B=$Particles.QualityLevelOpacity)
       )
       Spawn Group Mask=&SelectIntFromArray(Direct Array Index=$Particles.SpawnGroup, Int Selection Array=DataInterface() {propertyValues {IntData [0,0,0,0,1,1]}})
      )
     )
    )) {isSetParameters=true}
   /SetVariables(Particles.GlowRibbonWidth:NiagaraFloat=&Multiply_Float(
     A=$Particles.Wave
     B=&Lerp_Float(
      A=&FloatFromCurve(
       FloatCurve=DataInterface() {propertyValues {Curve {keys [
           {interpMode=RCIM_Cubic, tangentMode=RCTM_User, time=0, value=1}
           {arriveTangent=-2.0527, tangentMode=RCTM_User, time=0.2, value=0.6}
           {arriveTangent=1.9371, interpMode=RCIM_Cubic, leaveTangent=1.9371, tangentMode=RCTM_User, time=0.8, value=0.6}
           {tangentMode=RCTM_User, time=1, value=1}
          ]}}}
       CurveIndex=$Particles.Normalized Spawn Group
       Scale Curve=4
      )
      B=4
      Alpha=$System.AmplitudeValue
     )
    )) {isSetParameters=true}
   &DynamicMaterialParameters(
    Index 0 Param 3=&MakeCustomFloatFromBool(
     Bool=$System.bIsReactive
     True Float=$System.ReactiveValue
     False Float=&FloatFromCurve(
      FloatCurve=DataInterface() {propertyValues {Curve {keys [
          {interpMode=RCIM_Cubic, leaveTangent=4.8905, tangentMode=RCTM_Break, time=0, value=0.15}
          <at 0.2 0.9>
          {interpMode=RCIM_Cubic, leaveTangent=-1.0585, tangentMode=RCTM_Break, time=0.3, value=0.9}
          {arriveTangent=-0.6609, tangentMode=RCTM_User, time=1, value=0.15}
         ]}}}
      CurveIndex=$System.NormalizedLoopAge
     )
    )
    Index 0 Param 4=&MakeCustomFloatFromBool(Bool=$System.bIsReactive, True Float=1, False Float=0)
    Module.Index0Param0Write=true
    Module.Index0Param1Write=true
    Module.Index0Param2Write=true
    Module.Index0Param3Write=true
   )
   &ScaleColor(
    Scale Mode=::NewEnumerator1
    Scale RGBA=&MakeVector4_FromLinearColor(LinearColor=&FNTimeOfDayColor(Night Color=<rgba 0.4683 0.4038 0.9 1>, Morning Color=<rgba 1 0.6536 0.6237 1>, Day Color=<rgba 1 1 1 1>, Evening Color=<rgba 1 1 1 1>))
    ScaleRGBA=true
   )
  ]
  rendererClasses [&NiagaraSpriteRendererProperties, &NiagaraRibbonRendererProperties]
  renderers [
   {rendererClass=&NiagaraSpriteRendererProperties}
   {rendererClass=&NiagaraRibbonRendererProperties, rendererIndex=1}
   {rendererClass=&NiagaraRibbonRendererProperties, rendererIndex=2}
   {rendererClass=&NiagaraRibbonRendererProperties, rendererIndex=3}
  ]
 }
 Bone_Flow_Hip_Right {
  emitterSpawnScript [
   /SetVariables(Emitter.RandomSeed:NiagaraInt32=&RandomInt(Randomness Mode:ENiagaraRandomnessMode=::NewEnumerator0, Random Seed:NiagaraInt32=0, Fixed Random Seed:NiagaraBool=false)) {isSetParameters=true, module=SetVariables_#11}
   /SetVariables(Emitter.ParticlesPerGroup=6) {isSetParameters=true, module=SetVariables_#12}
  ]
  emitterUpdateScript [
   /SetVariables(Emitter.NumberOfGroups=1) {isSetParameters=true, module=SetVariables_#13}
   &EmitterState(Life Cycle Mode=::NewEnumerator1, Loop Behavior=::NewEnumerator1, Loop Duration Mode=::NewEnumerator1, UseLoopDelay=false)
   &SpawnBurst_Instantaneous(
    Spawn Count=&Multiply_Int(A=$Emitter.ParticlesPerGroup, B=$Emitter.NumberOfGroups)
    Use Spawn Probability=false
   )
  ]
  particleSpawnScript [
   /SetVariables(Particles.LoopCount:NiagaraInt32=$Emitter.LoopCount) {isSetParameters=true, module=SetVariables_#14}
   &FN_SetSpawnGroup(
    Inclusivity=true
    Normalize Spawn Group=true
    Particles Per Group:NiagaraInt32=$Emitter.ParticlesPerGroup
    Pattern=::NewEnumerator0
    Set Ribbon ID and Ribbon Link Order=true
   )
   /SetVariables(Particles.QualityLevelOpacity:NiagaraFloat=&Lerp_Float(
     A=1
     B=&FloatFromCurve(
      FloatCurve=DataInterface() {propertyValues {Curve {keys [
          {interpMode=RCIM_Cubic, tangentMode=RCTM_User, time=0.1, value=0}
          {arriveTangent=1.4286, time=0.4, value=1}
          {interpMode=RCIM_Cubic, tangentMode=RCTM_User, time=0.75, value=1}
          {tangentMode=RCTM_User, time=1, value=0}
         ]}}}
      CurveIndex=$Particles.Normalized Spawn Group
     )
     Alpha=&MakeFloatFromInt(Int=&SelectIntFromArray(Direct Array Index=$Engine.QualityLevel, Int Selection Array=DataInterface() {propertyValues {IntData [1,1,0,0,0]}}))
    )) {isSetParameters=true}
   &InitializeRibbon(
    Color=<rgba 1 1 1 0.33>
    Lifetime=5
    Module.Write Color=true
    Module.Write Lifetime=true
    Module.Write Mass=true
    Module.Write Material Random=true
    Module.Write Position=false
    Module.Write RibbonTwist=false
    Module.Write RibbonWidth=true
    Ribbon Width=8
   )
   &FN_FlipRibbonV() {enabled=false}
  ]
  particleUpdateScript [
   &ParticleState(Kill Particles When Lifetime Has Elapsed=false, Let Infinitely Lived Particles Die When Emitter Deactivates=true)
   /SetVariables(Particles.Wave:NiagaraFloat=&Waveform(
     Global Input Position=$Particles.Normalized Spawn Group
     Global Amplitude Scale=0.7
     "Waveform [1]"=::NewEnumerator0
     "[1] Phase Scale"=0.5
     "[1] Frequency"=1.5
     "[1] Amplitude Min/Max"=<0 1>
     "[1] Scale"=&Lerp_Float(A=0, B=1, Alpha=$System.AmplitudeValue)
    )) {isSetParameters=true}
   &ScaleRibbonWidth(Ribbon Width Scale=&FloatFromCurve(
     FloatCurve=DataInterface() {propertyValues {Curve {keys [
         {arriveTangent=6.0891, interpMode=RCIM_Cubic, leaveTangent=6.0891, tangentMode=RCTM_User, time=0, value=1}
         {arriveTangent=1.25, time=0.1, value=1}
         {interpMode=RCIM_Cubic, tangentMode=RCTM_User, time=0.9, value=1}
         {arriveTangent=-4.9474, tangentMode=RCTM_User, time=1, value=1}
        ]}}}
     CurveIndex=$Particles.Normalized Spawn Group
     Scale Curve=&Lerp_Float(A=1, B=1.25, Alpha=$System.AmplitudeValue)
    ))
   &FN_BoneFlow(
    Angle:NiagaraFloat=&FloatFromCurve(
     FloatCurve=DataInterface() {propertyValues {Curve {keys [
         <at 0 1 RCIM_Cubic>
         {arriveTangent=-0.75, interpMode=RCIM_Cubic, leaveTangent=-0.75, time=0.3, value=0.85}
         {arriveTangent=-0.6429, interpMode=RCIM_Cubic, leaveTangent=-0.6429, time=0.6, value=0.55}
         <at 1 0.4 RCIM_Cubic>
        ]}}}
     CurveIndex=$Particles.Normalized Spawn Group
     Scale Curve=0.256
    )
    BoneID:NiagaraInt32=&Add_Integer(A=$Particles.SpawnGroup, B=4)
    Height:NiagaraFloat=$Particles.Normalized Spawn Group
    Radius:NiagaraFloat=&FloatFromCurve(
     FloatCurve=DataInterface() {propertyValues {Curve {keys [
         {arriveTangent=31.4421, interpMode=RCIM_Cubic, leaveTangent=31.4421, tangentMode=RCTM_User, time=0, value=-0.5}
         {arriveTangent=6, interpMode=RCIM_Cubic, leaveTangent=6, time=0.2, value=4}
         {arriveTangent=-5.625, interpMode=RCIM_Cubic, leaveTangent=-5.625, time=0.5, value=2.5}
         {arriveTangent=-14.1598, interpMode=RCIM_Cubic, leaveTangent=-14.1598, tangentMode=RCTM_User, time=1, value=-0.5}
        ]}}}
     CurveIndex=$Particles.Normalized Spawn Group
     Scale Curve=&Add_Float(A=1, B=$Particles.Wave)
    )
    SkeletalMesh:NiagaraDataInterfaceSkeletalMesh=$System.BoneFlowSkeleMesh1
    Use Sockets Instead=true
    Write Position Attribute=true
   )
   /SetVariables(Particles.RibbonFacing=$Output.FN_BoneFlow.RotatedBiNormal) {enabled=false, isSetParameters=true, module=SetVariables_#15}
   &FN_RibbonMaterial(
    Color=<rgba 0 1 1 9>
    Opacity Pan U Offset:NiagaraFloat=&FN_SeededRandomRangeFloat(
     Minimum:NiagaraFloat=0
     Maximum:NiagaraFloat=2
     Random Seed:NiagaraInt32=&Add_Integer(A=$Particles.LoopCount, B=$Particles.SpawnGroup)
    )
    Opacity Pan U Speed:NiagaraFloat=&FN_SeededRandomRangeFloat(
     Minimum:NiagaraFloat=0.1364
     Maximum:NiagaraFloat=0.5182
     Random Seed:NiagaraInt32=&Add_Integer(A=$Particles.LoopCount, B=$Particles.SpawnGroup)
    )
    Opacity Scale U:NiagaraFloat=&FN_SeededRandomRangeFloat(
     Minimum:NiagaraFloat=0.4273
     Maximum:NiagaraFloat=1.0032
     Random Seed:NiagaraInt32=&Add_Integer(A=$Particles.LoopCount, B=$Particles.SpawnGroup)
    )
    Scale Alpha:NiagaraFloat=&FloatFromCurve(
     FloatCurve=DataInterface() {propertyValues {Curve {keys [<at -0.007 0.0112>,<at 0.2367 0.9629>,<at 0.7537 0.9995>,<at 1 0>]}}}
     Scale Curve=&FloatFromCurve(
      FloatCurve=DataInterface() {propertyValues {Curve {keys [<at 0.0007 0.022>,<at 0.2494 1.0225>,<at 0.8089 0.9748>,<at 1 0>]}}}
      CurveIndex=$Particles.Normalized Spawn Group
     )
    )
    UV Basis:NiagaraFloat=$Particles.Normalized Spawn Group
   ) {enabled=false}
   &FN_Sprite_Attributes(Color=<rgba 1 0.597 0 1>, Create Color Attribute=true, Create SpriteSize Attribute=true, SpriteSize=<3 3>, Write to Module Namespace=true) {enabled=false}
   &ScaleColor(
    Scale Alpha=&FloatFromCurve(
     FloatCurve=DataInterface() {propertyValues {Curve {keys [
         {arriveTangent=4.2704, interpMode=RCIM_Cubic, leaveTangent=4.2704, tangentMode=RCTM_User, time=0, value=0}
         {interpMode=RCIM_Cubic, tangentMode=RCTM_User, time=0.2, value=1}
         {arriveTangent=-1.25, interpMode=RCIM_Cubic, leaveTangent=-1.25, time=0.5, value=1}
         {arriveTangent=-0.4071, tangentMode=RCTM_User, time=1, value=0}
        ]}}}
     CurveIndex=$Particles.Normalized Spawn Group
     Scale Curve=&SelectFloatFromArray(Direct Array Index=$Engine.QualityLevel, Float Selection Array=DataInterface() {propertyValues {FloatData [0.4,0.6,1,1,1]}})
    )
    Scale RGB=&Lerp_Vector(
     A=<1 1 1>
     B=&MakeVectorFromLinearColorRGB(Color=&ColorFromCurve(
       ColorCurve=DataInterface() {propertyValues {
         AlphaCurve {keys [
           {interpMode=RCIM_Cubic, leaveTangentWeight=0.3333, time=0, value=1}
           {arriveTangentWeight=0.3333, interpMode=RCIM_Cubic, time=0.5009, value=1}
          ]}
         BlueCurve {keys [
           {interpMode=RCIM_Cubic, tangentMode=RCTM_User, time=0, value=0}
           {interpMode=RCIM_Cubic, leaveTangentWeight=0.1194, tangentMode=RCTM_User, time=0.2136, value=0}
           {
            arriveTangent=1.7758
            arriveTangentWeight=0.1474
            interpMode=RCIM_Cubic
            leaveTangent=1.7758
            leaveTangentWeight=0.0666
            time=0.4622
            value=0.7106
           }
           {arriveTangent=-0.2449, arriveTangentWeight=0.0666, interpMode=RCIM_Cubic, leaveTangent=-0.2449, time=0.7387, value=0.9324}
           {interpMode=RCIM_Cubic, tangentMode=RCTM_User, time=0.9608, value=0.5885}
          ]}
         GreenCurve {keys [
           {interpMode=RCIM_Cubic, tangentMode=RCTM_User, time=0, value=0.5781}
           {arriveTangent=-1.2509, interpMode=RCIM_Cubic, leaveTangent=-1.2509, leaveTangentWeight=0.1194, time=0.2136, value=0.8802}
           {
            arriveTangent=0.1488
            arriveTangentWeight=0.1474
            interpMode=RCIM_Cubic
            leaveTangent=0.1488
            leaveTangentWeight=0.0666
            time=0.4622
            value=0
           }
           {arriveTangent=0.5821, arriveTangentWeight=0.0666, interpMode=RCIM_Cubic, leaveTangent=0.5821, time=0.7387, value=0.9583}
           {interpMode=RCIM_Cubic, tangentMode=RCTM_User, time=0.9608, value=0.2903}
          ]}
         RedCurve {keys [
           {interpMode=RCIM_Cubic, tangentMode=RCTM_User, time=0, value=0.0241}
           {arriveTangent=2.1116, interpMode=RCIM_Cubic, leaveTangent=2.1116, leaveTangentWeight=0.1194, time=0.2136, value=0.7506}
           {arriveTangentWeight=0.2294, interpMode=RCIM_Cubic, leaveTangentWeight=0.1036, tangentMode=RCTM_User, time=0.4622, value=1}
           {arriveTangent=-2.0054, arriveTangentWeight=0.0666, interpMode=RCIM_Cubic, leaveTangent=-2.0054, time=0.7387, value=0}
           {interpMode=RCIM_Cubic, tangentMode=RCTM_User, time=0.9608, value=0}
          ]}
        }}
       CurveIndex=&Multiply_Float(A=$Particles.Normalized Spawn Group, B=$System.Timeline.Time)
      ))
     Alpha=$System.AmplitudeValue
    )
    ScaleA=true
    ScaleRGB=true
   ) {module=ScaleColor001}
   /SetVariables(Particles.GlowtubeColor:LinearColor=&MakeLinearColorFromVectorAndFloat(
     "Vector (RGB)"=<0.2 0.2 0.2>
     "Float (Alpha)"=&FloatFromCurve(
      FloatCurve=DataInterface() {propertyValues {Curve {keys [
          {arriveTangent=21.1196, interpMode=RCIM_Cubic, leaveTangent=21.1196, tangentMode=RCTM_User, time=0, value=0}
          {interpMode=RCIM_Cubic, tangentMode=RCTM_User, time=0.05, value=1}
          <at 0.5 0 RCIM_Cubic>
          <at 0.95 1 RCIM_Cubic>
          {arriveTangent=-20.6713, interpMode=RCIM_Cubic, leaveTangent=-20.6713, tangentMode=RCTM_User, time=1, value=0}
         ]}}}
      CurveIndex=$Particles.Normalized Spawn Group
      Scale Curve=&MaskFloatBySpawnGroup(
       Masked Float=&Multiply_Float(A=0.25, B=$Particles.QualityLevelOpacity)
       Passthrough Float=&FloatFromCurve(
        FloatCurve=DataInterface() {propertyValues {Curve {keys [{tangentMode=RCTM_User, time=0, value=1}, {tangentMode=RCTM_User, time=0.7, value=0}]}}}
        CurveIndex=$Particles.Normalized Spawn Group
        Scale Curve=&Multiply_Float(A=0.25, B=$Particles.QualityLevelOpacity)
       )
       Spawn Group Mask=&SelectIntFromArray(Direct Array Index=$Particles.SpawnGroup, Int Selection Array=DataInterface() {propertyValues {IntData [0,0,0,0,1,1]}})
      )
     )
    )) {isSetParameters=true}
   /SetVariables(Particles.GlowRibbonWidth:NiagaraFloat=&Multiply_Float(
     A=$Particles.Wave
     B=&Lerp_Float(
      A=&FloatFromCurve(
       FloatCurve=DataInterface() {propertyValues {Curve {keys [
           {interpMode=RCIM_Cubic, tangentMode=RCTM_User, time=0, value=1}
           {arriveTangent=-2.0527, tangentMode=RCTM_User, time=0.2, value=0.6}
           {arriveTangent=1.9371, interpMode=RCIM_Cubic, leaveTangent=1.9371, tangentMode=RCTM_User, time=0.8, value=0.6}
           {tangentMode=RCTM_User, time=1, value=1}
          ]}}}
       CurveIndex=$Particles.Normalized Spawn Group
       Scale Curve=4
      )
      B=4
      Alpha=$System.AmplitudeValue
     )
    )) {isSetParameters=true}
   &DynamicMaterialParameters(
    Index 0 Param 3=&MakeCustomFloatFromBool(
     Bool=$System.bIsReactive
     True Float=$System.ReactiveValue
     False Float=&FloatFromCurve(
      FloatCurve=DataInterface() {propertyValues {Curve {keys [
          {interpMode=RCIM_Cubic, leaveTangent=4.8905, tangentMode=RCTM_Break, time=0, value=0.15}
          <at 0.2 0.9>
          {interpMode=RCIM_Cubic, leaveTangent=-1.0585, tangentMode=RCTM_Break, time=0.3, value=0.9}
          {arriveTangent=-0.6609, tangentMode=RCTM_User, time=1, value=0.15}
         ]}}}
      CurveIndex=$System.NormalizedLoopAge
     )
    )
    Index 0 Param 4=&MakeCustomFloatFromBool(Bool=$System.bIsReactive, True Float=1, False Float=0)
    Module.Index0Param0Write=true
    Module.Index0Param1Write=true
    Module.Index0Param2Write=true
    Module.Index0Param3Write=true
   )
   &ScaleColor(
    Scale Mode=::NewEnumerator1
    Scale RGBA=&MakeVector4_FromLinearColor(LinearColor=&FNTimeOfDayColor(Night Color=<rgba 0.4683 0.4038 0.9 1>, Morning Color=<rgba 1 0.6536 0.6237 1>, Day Color=<rgba 1 1 1 1>, Evening Color=<rgba 1 1 1 1>))
    ScaleRGBA=true
   )
  ]
  rendererClasses [&NiagaraSpriteRendererProperties, &NiagaraRibbonRendererProperties]
  renderers [
   {rendererClass=&NiagaraSpriteRendererProperties}
   {rendererClass=&NiagaraRibbonRendererProperties, rendererIndex=1}
   {rendererClass=&NiagaraRibbonRendererProperties, rendererIndex=2}
   {rendererClass=&NiagaraRibbonRendererProperties, rendererIndex=3}
  ]
 }
 Bone_Flow_Shoulders {
  emitterSpawnScript [
   /SetVariables(Emitter.RandomSeed:NiagaraInt32=&RandomInt(Randomness Mode:ENiagaraRandomnessMode=::NewEnumerator0, Random Seed:NiagaraInt32=0, Fixed Random Seed:NiagaraBool=false)) {isSetParameters=true, module=SetVariables_#11}
   /SetVariables(Emitter.ParticlesPerGroup=6) {isSetParameters=true, module=SetVariables_#12}
  ]
  emitterUpdateScript [
   /SetVariables(Emitter.NumberOfGroups=2) {isSetParameters=true, module=SetVariables_#13}
   &EmitterState(Life Cycle Mode=::NewEnumerator1, Loop Behavior=::NewEnumerator1, Loop Duration Mode=::NewEnumerator1, UseLoopDelay=false)
   &SpawnBurst_Instantaneous(
    Spawn Count=&Multiply_Int(A=$Emitter.ParticlesPerGroup, B=$Emitter.NumberOfGroups)
    Use Spawn Probability=false
   )
  ]
  particleSpawnScript [
   /SetVariables(Particles.LoopCount:NiagaraInt32=$Emitter.LoopCount) {isSetParameters=true, module=SetVariables_#14}
   &FN_SetSpawnGroup(
    Inclusivity=true
    Normalize Spawn Group=true
    Particles Per Group:NiagaraInt32=$Emitter.ParticlesPerGroup
    Pattern=::NewEnumerator0
    Set Ribbon ID and Ribbon Link Order=true
   )
   /SetVariables(Particles.QualityLevelOpacity:NiagaraFloat=&Lerp_Float(
     A=1
     B=&FloatFromCurve(
      FloatCurve=DataInterface() {propertyValues {Curve {keys [
          {interpMode=RCIM_Cubic, tangentMode=RCTM_User, time=0.1, value=0}
          {arriveTangent=1.4286, time=0.4, value=1}
          {interpMode=RCIM_Cubic, tangentMode=RCTM_User, time=0.75, value=1}
          {tangentMode=RCTM_User, time=1, value=0}
         ]}}}
      CurveIndex=$Particles.Normalized Spawn Group
     )
     Alpha=&MakeFloatFromInt(Int=&SelectIntFromArray(Direct Array Index=$Engine.QualityLevel, Int Selection Array=DataInterface() {propertyValues {IntData [1,1,0,0,0]}}))
    )) {isSetParameters=true}
   &InitializeRibbon(
    Color=<rgba 1 1 1 0.33>
    Lifetime=5
    Module.Write Color=true
    Module.Write Lifetime=true
    Module.Write Mass=true
    Module.Write Material Random=true
    Module.Write Position=false
    Module.Write RibbonTwist=false
    Module.Write RibbonWidth=true
    Ribbon Width=8
   )
   &FN_FlipRibbonV() {enabled=false}
  ]
  particleUpdateScript [
   &ParticleState(Kill Particles When Lifetime Has Elapsed=false, Let Infinitely Lived Particles Die When Emitter Deactivates=true)
   /SetVariables(Particles.Wave:NiagaraFloat=&Waveform(
     Global Input Position=$Particles.Normalized Spawn Group
     Global Amplitude Scale=0.65
     "Waveform [1]"=::NewEnumerator0
     "[1] Phase Scale"=0.5
     "[1] Frequency"=1.5
     "[1] Amplitude Min/Max"=<0 1>
     "[1] Scale"=&Lerp_Float(A=0, B=1, Alpha=$System.AmplitudeValue)
     "[2] Scale"=&Lerp_Float(A=0, B=1, Alpha=$System.AmplitudeValue)
    )) {isSetParameters=true}
   &ScaleRibbonWidth(Ribbon Width Scale=&FloatFromCurve(
     FloatCurve=DataInterface() {propertyValues {Curve {keys [
         {arriveTangent=6.0891, interpMode=RCIM_Cubic, leaveTangent=6.0891, tangentMode=RCTM_User, time=0, value=1}
         {arriveTangent=1.25, time=0.1, value=1}
         {interpMode=RCIM_Cubic, tangentMode=RCTM_User, time=0.9, value=1}
         {arriveTangent=-4.9474, tangentMode=RCTM_User, time=1, value=1}
        ]}}}
     CurveIndex=$Particles.Normalized Spawn Group
     Scale Curve=&Lerp_Float(A=1, B=1.25, Alpha=$System.AmplitudeValue)
    ))
   &FN_BoneFlow(
    Angle:NiagaraFloat=&FloatFromCurve(
     FloatCurve=DataInterface() {propertyValues {Curve {keys [
         {arriveTangent=11.4328, interpMode=RCIM_Cubic, leaveTangent=11.4328, tangentMode=RCTM_User, time=0.05, value=0}
         <at 0.5 1 RCIM_Cubic>
         {arriveTangent=-10.1631, interpMode=RCIM_Cubic, leaveTangent=-10.1631, tangentMode=RCTM_User, time=0.95, value=0}
        ]}}}
     CurveIndex=$Particles.Normalized Spawn Group
     Scale Curve=0.256
    )
    BoneID:NiagaraInt32=&ModuloInt(A=$Particles.SpawnGroup, B=2)
    Height:NiagaraFloat=$Particles.Normalized Spawn Group
    Radius:NiagaraFloat=&FloatFromCurve(
     FloatCurve=DataInterface() {propertyValues {Curve {keys [
         {arriveTangent=108.6975, interpMode=RCIM_Cubic, leaveTangent=108.6975, tangentMode=RCTM_User, time=0.05, value=0}
         <at 0.5 7.25 RCIM_Cubic>
         {arriveTangent=-97.1793, interpMode=RCIM_Cubic, leaveTangent=-97.1793, tangentMode=RCTM_User, time=0.95, value=0}
        ]}}}
     CurveIndex=$Particles.Normalized Spawn Group
     Scale Curve=&Add_Float(A=1, B=$Particles.Wave)
    )
    SkeletalMesh:NiagaraDataInterfaceSkeletalMesh=$System.BoneFlowSkeleMesh1
    Use Sockets Instead=true
    Write Position Attribute=true
   )
   /SetVariables(Particles.RibbonFacing=$Output.FN_BoneFlow.RotatedBiNormal) {enabled=false, isSetParameters=true, module=SetVariables_#15}
   &FN_RibbonMaterial(
    Color=<rgba 0 1 1 9>
    Opacity Pan U Offset:NiagaraFloat=&FN_SeededRandomRangeFloat(
     Minimum:NiagaraFloat=0
     Maximum:NiagaraFloat=2
     Random Seed:NiagaraInt32=&Add_Integer(A=$Particles.LoopCount, B=$Particles.SpawnGroup)
    )
    Opacity Pan U Speed:NiagaraFloat=&FN_SeededRandomRangeFloat(
     Minimum:NiagaraFloat=0.1364
     Maximum:NiagaraFloat=0.5182
     Random Seed:NiagaraInt32=&Add_Integer(A=$Particles.LoopCount, B=$Particles.SpawnGroup)
    )
    Opacity Scale U:NiagaraFloat=&FN_SeededRandomRangeFloat(
     Minimum:NiagaraFloat=0.4273
     Maximum:NiagaraFloat=1.0032
     Random Seed:NiagaraInt32=&Add_Integer(A=$Particles.LoopCount, B=$Particles.SpawnGroup)
    )
    Scale Alpha:NiagaraFloat=&FloatFromCurve(
     FloatCurve=DataInterface() {propertyValues {Curve {keys [<at -0.007 0.0112>,<at 0.2367 0.9629>,<at 0.7537 0.9995>,<at 1 0>]}}}
     Scale Curve=&FloatFromCurve(
      FloatCurve=DataInterface() {propertyValues {Curve {keys [<at 0.0007 0.022>,<at 0.2494 1.0225>,<at 0.8089 0.9748>,<at 1 0>]}}}
      CurveIndex=$Particles.Normalized Spawn Group
     )
    )
    UV Basis:NiagaraFloat=$Particles.Normalized Spawn Group
   ) {enabled=false}
   &FN_Sprite_Attributes(Color=<rgba 1 0.597 0 1>, Create Color Attribute=true, Create SpriteSize Attribute=true, SpriteSize=<3 3>, Write to Module Namespace=true) {enabled=false}
   &ScaleColor(
    Scale Alpha=&SelectFloatFromArray(Direct Array Index=$Engine.QualityLevel, Float Selection Array=DataInterface() {propertyValues {FloatData [0.4,0.6,1,1,1]}})
    Scale RGB=&Lerp_Vector(
     A=<1 1 1>
     B=&MakeVectorFromLinearColorRGB(Color=&ColorFromCurve(
       ColorCurve=DataInterface() {propertyValues {
         AlphaCurve {keys [
           {interpMode=RCIM_Cubic, leaveTangentWeight=0.3333, time=0, value=1}
           {arriveTangentWeight=0.3333, interpMode=RCIM_Cubic, time=0.5009, value=1}
          ]}
         BlueCurve {keys [
           {interpMode=RCIM_Cubic, tangentMode=RCTM_User, time=0, value=0}
           {interpMode=RCIM_Cubic, leaveTangentWeight=0.1194, tangentMode=RCTM_User, time=0.2136, value=0}
           {
            arriveTangent=1.7758
            arriveTangentWeight=0.1474
            interpMode=RCIM_Cubic
            leaveTangent=1.7758
            leaveTangentWeight=0.0666
            time=0.4622
            value=0.7106
           }
           {arriveTangent=-0.2449, arriveTangentWeight=0.0666, interpMode=RCIM_Cubic, leaveTangent=-0.2449, time=0.7387, value=0.9324}
           {interpMode=RCIM_Cubic, tangentMode=RCTM_User, time=0.9608, value=0.5885}
          ]}
         GreenCurve {keys [
           {interpMode=RCIM_Cubic, tangentMode=RCTM_User, time=0, value=0.5781}
           {arriveTangent=-1.2509, interpMode=RCIM_Cubic, leaveTangent=-1.2509, leaveTangentWeight=0.1194, time=0.2136, value=0.8802}
           {
            arriveTangent=0.1488
            arriveTangentWeight=0.1474
            interpMode=RCIM_Cubic
            leaveTangent=0.1488
            leaveTangentWeight=0.0666
            time=0.4622
            value=0
           }
           {arriveTangent=0.5821, arriveTangentWeight=0.0666, interpMode=RCIM_Cubic, leaveTangent=0.5821, time=0.7387, value=0.9583}
           {interpMode=RCIM_Cubic, tangentMode=RCTM_User, time=0.9608, value=0.2903}
          ]}
         RedCurve {keys [
           {interpMode=RCIM_Cubic, tangentMode=RCTM_User, time=0, value=0.0241}
           {arriveTangent=2.1116, interpMode=RCIM_Cubic, leaveTangent=2.1116, leaveTangentWeight=0.1194, time=0.2136, value=0.7506}
           {arriveTangentWeight=0.2294, interpMode=RCIM_Cubic, leaveTangentWeight=0.1036, tangentMode=RCTM_User, time=0.4622, value=1}
           {arriveTangent=-2.0054, arriveTangentWeight=0.0666, interpMode=RCIM_Cubic, leaveTangent=-2.0054, time=0.7387, value=0}
           {interpMode=RCIM_Cubic, tangentMode=RCTM_User, time=0.9608, value=0}
          ]}
        }}
       CurveIndex=&Multiply_Float(A=$Particles.Normalized Spawn Group, B=$System.Timeline.Time)
      ))
     Alpha=$System.AmplitudeValue
    )
    ScaleA=true
    ScaleRGB=true
   ) {module=ScaleColor002}
   /SetVariables(Particles.GlowtubeColor:LinearColor=&MakeLinearColorFromVectorAndFloat(
     "Vector (RGB)"=<0.2 0.2 0.2>
     "Float (Alpha)"=&FloatFromCurve(
      FloatCurve=DataInterface() {propertyValues {Curve {keys [
          {arriveTangent=21.1196, interpMode=RCIM_Cubic, leaveTangent=21.1196, tangentMode=RCTM_User, time=0, value=0}
          {interpMode=RCIM_Cubic, tangentMode=RCTM_User, time=0.05, value=1}
          <at 0.5 0 RCIM_Cubic>
          <at 0.95 1 RCIM_Cubic>
          {arriveTangent=-20.6713, interpMode=RCIM_Cubic, leaveTangent=-20.6713, tangentMode=RCTM_User, time=1, value=0}
         ]}}}
      CurveIndex=$Particles.Normalized Spawn Group
      Scale Curve=&MaskFloatBySpawnGroup(
       Masked Float=&Multiply_Float(A=0.25, B=$Particles.QualityLevelOpacity)
       Passthrough Float=&FloatFromCurve(
        FloatCurve=DataInterface() {propertyValues {Curve {keys [{tangentMode=RCTM_User, time=0, value=1}, {tangentMode=RCTM_User, time=0.7, value=0}]}}}
        CurveIndex=$Particles.Normalized Spawn Group
        Scale Curve=&Multiply_Float(A=0.25, B=$Particles.QualityLevelOpacity)
       )
       Spawn Group Mask=&SelectIntFromArray(Direct Array Index=$Particles.SpawnGroup, Int Selection Array=DataInterface() {propertyValues {IntData [0,0,0,0,1,1]}})
      )
     )
    )) {isSetParameters=true}
   /SetVariables(Particles.GlowRibbonWidth:NiagaraFloat=&Multiply_Float(
     A=$Particles.Wave
     B=&Lerp_Float(
      A=&FloatFromCurve(
       FloatCurve=DataInterface() {propertyValues {Curve {keys [
           {interpMode=RCIM_Cubic, tangentMode=RCTM_User, time=0, value=1}
           {arriveTangent=-2.0527, tangentMode=RCTM_User, time=0.2, value=0.6}
           {arriveTangent=1.9371, interpMode=RCIM_Cubic, leaveTangent=1.9371, tangentMode=RCTM_User, time=0.8, value=0.6}
           {tangentMode=RCTM_User, time=1, value=1}
          ]}}}
       CurveIndex=$Particles.Normalized Spawn Group
       Scale Curve=4
      )
      B=4
      Alpha=$System.AmplitudeValue
     )
    )) {isSetParameters=true}
   &DynamicMaterialParameters(
    Index 0 Param 3=&MakeCustomFloatFromBool(
     Bool=$System.bIsReactive
     True Float=$System.ReactiveValue
     False Float=&FloatFromCurve(
      FloatCurve=DataInterface() {propertyValues {Curve {keys [
          {interpMode=RCIM_Cubic, leaveTangent=4.8905, tangentMode=RCTM_Break, time=0, value=0.15}
          <at 0.2 0.9>
          {interpMode=RCIM_Cubic, leaveTangent=-1.0585, tangentMode=RCTM_Break, time=0.3, value=0.9}
          {arriveTangent=-0.6609, tangentMode=RCTM_User, time=1, value=0.15}
         ]}}}
      CurveIndex=$System.NormalizedLoopAge
     )
    )
    Index 0 Param 4=&MakeCustomFloatFromBool(Bool=$System.bIsReactive, True Float=1, False Float=0)
    Module.Index0Param0Write=true
    Module.Index0Param1Write=true
    Module.Index0Param2Write=true
    Module.Index0Param3Write=true
   )
   &ScaleColor(
    Scale Mode=::NewEnumerator1
    Scale RGBA=&MakeVector4_FromLinearColor(LinearColor=&FNTimeOfDayColor(Night Color=<rgba 0.4683 0.4038 0.9 1>, Morning Color=<rgba 1 0.6536 0.6237 1>, Day Color=<rgba 1 1 1 1>, Evening Color=<rgba 1 1 1 1>))
    ScaleRGBA=true
   )
  ]
  rendererClasses [&NiagaraSpriteRendererProperties, &NiagaraRibbonRendererProperties]
  renderers [
   {rendererClass=&NiagaraSpriteRendererProperties}
   {rendererClass=&NiagaraRibbonRendererProperties, rendererIndex=1}
   {rendererClass=&NiagaraRibbonRendererProperties, rendererIndex=2}
   {rendererClass=&NiagaraRibbonRendererProperties, rendererIndex=3}
  ]
 }
 Bone_Flow_Waist {
  emitterSpawnScript [
   /SetVariables(Emitter.RandomSeed:NiagaraInt32=&RandomInt(Randomness Mode:ENiagaraRandomnessMode=::NewEnumerator0, Random Seed:NiagaraInt32=0, Fixed Random Seed:NiagaraBool=false)) {isSetParameters=true, module=SetVariables_#11}
   /SetVariables(Emitter.ParticlesPerGroup=6) {isSetParameters=true, module=SetVariables_#12}
  ]
  emitterUpdateScript [
   /SetVariables(Emitter.NumberOfGroups=2) {isSetParameters=true, module=SetVariables_#13}
   &EmitterState(Life Cycle Mode=::NewEnumerator1, Loop Behavior=::NewEnumerator1, Loop Duration Mode=::NewEnumerator1, UseLoopDelay=false)
   &SpawnBurst_Instantaneous(
    Spawn Count=&Multiply_Int(A=$Emitter.ParticlesPerGroup, B=$Emitter.NumberOfGroups)
    Use Spawn Probability=false
   )
  ]
  particleSpawnScript [
   /SetVariables(Particles.LoopCount:NiagaraInt32=$Emitter.LoopCount) {isSetParameters=true, module=SetVariables_#14}
   &FN_SetSpawnGroup(
    Inclusivity=true
    Normalize Spawn Group=true
    Particles Per Group:NiagaraInt32=$Emitter.ParticlesPerGroup
    Pattern=::NewEnumerator0
    Set Ribbon ID and Ribbon Link Order=true
   )
   /SetVariables(Particles.QualityLevelOpacity:NiagaraFloat=&Lerp_Float(
     A=1
     B=&FloatFromCurve(
      FloatCurve=DataInterface() {propertyValues {Curve {keys [
          {interpMode=RCIM_Cubic, tangentMode=RCTM_User, time=0.1, value=0}
          {arriveTangent=1.4286, time=0.4, value=1}
          {interpMode=RCIM_Cubic, tangentMode=RCTM_User, time=0.75, value=1}
          {tangentMode=RCTM_User, time=1, value=0}
         ]}}}
      CurveIndex=$Particles.Normalized Spawn Group
     )
     Alpha=&MakeFloatFromInt(Int=&SelectIntFromArray(Direct Array Index=$Engine.QualityLevel, Int Selection Array=DataInterface() {propertyValues {IntData [1,1,0,0,0]}}))
    )) {isSetParameters=true}
   &InitializeRibbon(
    Color=<rgba 1 1 1 0.33>
    Lifetime=5
    Module.Write Color=true
    Module.Write Lifetime=true
    Module.Write Mass=true
    Module.Write Material Random=true
    Module.Write Position=false
    Module.Write RibbonTwist=false
    Module.Write RibbonWidth=true
    Ribbon Width=8
   )
   &FN_FlipRibbonV() {enabled=false}
  ]
  particleUpdateScript [
   &ParticleState(Kill Particles When Lifetime Has Elapsed=false, Let Infinitely Lived Particles Die When Emitter Deactivates=true)
   /SetVariables(Particles.Wave:NiagaraFloat=&Waveform(
     Global Input Position=$Particles.Normalized Spawn Group
     Global Amplitude Scale=0.65
     "Waveform [1]"=::NewEnumerator0
     "[1] Phase Scale"=0.5
     "[1] Frequency"=1.5
     "[1] Amplitude Min/Max"=<0 1>
     "[1] Scale"=&Lerp_Float(A=0, B=1, Alpha=$System.AmplitudeValue)
    )) {isSetParameters=true}
   &ScaleRibbonWidth(Ribbon Width Scale=&FloatFromCurve(
     FloatCurve=DataInterface() {propertyValues {Curve {keys [
         {arriveTangent=6.0891, interpMode=RCIM_Cubic, leaveTangent=6.0891, tangentMode=RCTM_User, time=0, value=1}
         {arriveTangent=1.25, time=0.1, value=1}
         {interpMode=RCIM_Cubic, tangentMode=RCTM_User, time=0.9, value=1}
         {arriveTangent=-4.9474, tangentMode=RCTM_User, time=1, value=1}
        ]}}}
     CurveIndex=$Particles.Normalized Spawn Group
     Scale Curve=&Lerp_Float(A=1, B=1.25, Alpha=$System.AmplitudeValue)
    ))
   &FN_BoneFlow(
    Angle:NiagaraFloat=&FloatFromCurve(
     FloatCurve=DataInterface() {propertyValues {Curve {keys [
         {arriveTangent=9.4726, interpMode=RCIM_Cubic, leaveTangent=9.4726, tangentMode=RCTM_User, time=0.05, value=0}
         <at 0.5 1 RCIM_Cubic>
         {arriveTangent=-6.9466, interpMode=RCIM_Cubic, leaveTangent=-6.9466, tangentMode=RCTM_User, time=0.95, value=0}
        ]}}}
     CurveIndex=$Particles.Normalized Spawn Group
     Scale Curve=0.256
    )
    BoneID:NiagaraInt32=&Add_Integer(A=$Particles.SpawnGroup, B=2)
    Height:NiagaraFloat=$Particles.Normalized Spawn Group
    Radius:NiagaraFloat=&FloatFromCurve(
     FloatCurve=DataInterface() {propertyValues {Curve {keys [
         {arriveTangent=48.2538, interpMode=RCIM_Cubic, leaveTangent=48.2538, tangentMode=RCTM_User, time=0.05, value=0}
         <at 0.5 3.5 RCIM_Cubic>
         {arriveTangent=-46.896, interpMode=RCIM_Cubic, leaveTangent=-46.896, tangentMode=RCTM_User, time=0.95, value=0}
        ]}}}
     CurveIndex=$Particles.Normalized Spawn Group
     Scale Curve=&Add_Float(A=1, B=$Particles.Wave)
    )
    SkeletalMesh:NiagaraDataInterfaceSkeletalMesh=$System.BoneFlowSkeleMesh1
    Use Sockets Instead=true
    Write Position Attribute=true
   )
   /SetVariables(Particles.RibbonFacing=$Output.FN_BoneFlow.RotatedBiNormal) {enabled=false, isSetParameters=true, module=SetVariables_#15}
   &FN_RibbonMaterial(
    Color=<rgba 0 1 1 9>
    Opacity Pan U Offset:NiagaraFloat=&FN_SeededRandomRangeFloat(
     Minimum:NiagaraFloat=0
     Maximum:NiagaraFloat=2
     Random Seed:NiagaraInt32=&Add_Integer(A=$Particles.LoopCount, B=$Particles.SpawnGroup)
    )
    Opacity Pan U Speed:NiagaraFloat=&FN_SeededRandomRangeFloat(
     Minimum:NiagaraFloat=0.1364
     Maximum:NiagaraFloat=0.5182
     Random Seed:NiagaraInt32=&Add_Integer(A=$Particles.LoopCount, B=$Particles.SpawnGroup)
    )
    Opacity Scale U:NiagaraFloat=&FN_SeededRandomRangeFloat(
     Minimum:NiagaraFloat=0.4273
     Maximum:NiagaraFloat=1.0032
     Random Seed:NiagaraInt32=&Add_Integer(A=$Particles.LoopCount, B=$Particles.SpawnGroup)
    )
    Scale Alpha:NiagaraFloat=&FloatFromCurve(
     FloatCurve=DataInterface() {propertyValues {Curve {keys [<at -0.007 0.0112>,<at 0.2367 0.9629>,<at 0.7537 0.9995>,<at 1 0>]}}}
     Scale Curve=&FloatFromCurve(
      FloatCurve=DataInterface() {propertyValues {Curve {keys [<at 0.0007 0.022>,<at 0.2494 1.0225>,<at 0.8089 0.9748>,<at 1 0>]}}}
      CurveIndex=$Particles.Normalized Spawn Group
     )
    )
    UV Basis:NiagaraFloat=$Particles.Normalized Spawn Group
   ) {enabled=false}
   &FN_Sprite_Attributes(Color=<rgba 1 0.597 0 1>, Create Color Attribute=true, Create SpriteSize Attribute=true, SpriteSize=<3 3>, Write to Module Namespace=true) {enabled=false}
   &ScaleColor(
    Scale Alpha=&FloatFromCurve(
     FloatCurve=DataInterface() {propertyValues {Curve {keys [{interpMode=RCIM_Cubic, tangentMode=RCTM_User, time=0, value=1}, {interpMode=RCIM_Cubic, tangentMode=RCTM_User, time=1, value=0}]}}}
     CurveIndex=$Particles.Normalized Spawn Group
     Scale Curve=&SelectFloatFromArray(Direct Array Index=$Engine.QualityLevel, Float Selection Array=DataInterface() {propertyValues {FloatData [0.4,0.6,1,1,1]}})
    )
    Scale RGB=&Lerp_Vector(
     A=<1 1 1>
     B=&MakeVectorFromLinearColorRGB(Color=&ColorFromCurve(
       ColorCurve=DataInterface() {propertyValues {
         AlphaCurve {keys [
           {interpMode=RCIM_Cubic, leaveTangentWeight=0.3333, time=0, value=1}
           {arriveTangentWeight=0.3333, interpMode=RCIM_Cubic, time=0.5009, value=1}
          ]}
         BlueCurve {keys [
           {interpMode=RCIM_Cubic, tangentMode=RCTM_User, time=0, value=0}
           {interpMode=RCIM_Cubic, leaveTangentWeight=0.1194, tangentMode=RCTM_User, time=0.2136, value=0}
           {
            arriveTangent=1.7758
            arriveTangentWeight=0.1474
            interpMode=RCIM_Cubic
            leaveTangent=1.7758
            leaveTangentWeight=0.0666
            time=0.4622
            value=0.7106
           }
           {arriveTangent=-0.2449, arriveTangentWeight=0.0666, interpMode=RCIM_Cubic, leaveTangent=-0.2449, time=0.7387, value=0.9324}
           {interpMode=RCIM_Cubic, tangentMode=RCTM_User, time=0.9608, value=0.5885}
          ]}
         GreenCurve {keys [
           {interpMode=RCIM_Cubic, tangentMode=RCTM_User, time=0, value=0.5781}
           {arriveTangent=-1.2509, interpMode=RCIM_Cubic, leaveTangent=-1.2509, leaveTangentWeight=0.1194, time=0.2136, value=0.8802}
           {
            arriveTangent=0.1488
            arriveTangentWeight=0.1474
            interpMode=RCIM_Cubic
            leaveTangent=0.1488
            leaveTangentWeight=0.0666
            time=0.4622
            value=0
           }
           {arriveTangent=0.5821, arriveTangentWeight=0.0666, interpMode=RCIM_Cubic, leaveTangent=0.5821, time=0.7387, value=0.9583}
           {interpMode=RCIM_Cubic, tangentMode=RCTM_User, time=0.9608, value=0.2903}
          ]}
         RedCurve {keys [
           {interpMode=RCIM_Cubic, tangentMode=RCTM_User, time=0, value=0.0241}
           {arriveTangent=2.1116, interpMode=RCIM_Cubic, leaveTangent=2.1116, leaveTangentWeight=0.1194, time=0.2136, value=0.7506}
           {arriveTangentWeight=0.2294, interpMode=RCIM_Cubic, leaveTangentWeight=0.1036, tangentMode=RCTM_User, time=0.4622, value=1}
           {arriveTangent=-2.0054, arriveTangentWeight=0.0666, interpMode=RCIM_Cubic, leaveTangent=-2.0054, time=0.7387, value=0}
           {interpMode=RCIM_Cubic, tangentMode=RCTM_User, time=0.9608, value=0}
          ]}
        }}
       CurveIndex=&Multiply_Float(A=$Particles.Normalized Spawn Group, B=$System.Timeline.Time)
      ))
     Alpha=$System.AmplitudeValue
    )
    ScaleA=true
    ScaleRGB=true
   ) {module=ScaleColor001}
   /SetVariables(Particles.GlowtubeColor:LinearColor=&MakeLinearColorFromVectorAndFloat(
     "Vector (RGB)"=<0.2 0.2 0.2>
     "Float (Alpha)"=&FloatFromCurve(
      FloatCurve=DataInterface() {propertyValues {Curve {keys [
          {arriveTangent=21.1196, interpMode=RCIM_Cubic, leaveTangent=21.1196, tangentMode=RCTM_User, time=0, value=0}
          {interpMode=RCIM_Cubic, tangentMode=RCTM_User, time=0.05, value=1}
          <at 0.5 0 RCIM_Cubic>
          <at 0.95 1 RCIM_Cubic>
          {arriveTangent=-20.6713, interpMode=RCIM_Cubic, leaveTangent=-20.6713, tangentMode=RCTM_User, time=1, value=0}
         ]}}}
      CurveIndex=$Particles.Normalized Spawn Group
      Scale Curve=&MaskFloatBySpawnGroup(
       Masked Float=&Multiply_Float(A=0.25, B=$Particles.QualityLevelOpacity)
       Passthrough Float=&FloatFromCurve(
        FloatCurve=DataInterface() {propertyValues {Curve {keys [{tangentMode=RCTM_User, time=0, value=1}, {tangentMode=RCTM_User, time=0.7, value=0}]}}}
        CurveIndex=$Particles.Normalized Spawn Group
        Scale Curve=&Multiply_Float(A=0.25, B=$Particles.QualityLevelOpacity)
       )
       Spawn Group Mask=&SelectIntFromArray(Direct Array Index=$Particles.SpawnGroup, Int Selection Array=DataInterface() {propertyValues {IntData [0,0,0,0,1,1]}})
      )
     )
    )) {isSetParameters=true}
   /SetVariables(Particles.GlowRibbonWidth:NiagaraFloat=&Multiply_Float(
     A=$Particles.Wave
     B=&Lerp_Float(
      A=&FloatFromCurve(
       FloatCurve=DataInterface() {propertyValues {Curve {keys [
           {interpMode=RCIM_Cubic, tangentMode=RCTM_User, time=0, value=1}
           {arriveTangent=-2.0527, tangentMode=RCTM_User, time=0.2, value=0.6}
           {arriveTangent=1.9371, interpMode=RCIM_Cubic, leaveTangent=1.9371, tangentMode=RCTM_User, time=0.8, value=0.6}
           {tangentMode=RCTM_User, time=1, value=1}
          ]}}}
       CurveIndex=$Particles.Normalized Spawn Group
       Scale Curve=4
      )
      B=4
      Alpha=$System.AmplitudeValue
     )
    )) {isSetParameters=true}
   &DynamicMaterialParameters(
    Index 0 Param 3=&MakeCustomFloatFromBool(
     Bool=$System.bIsReactive
     True Float=$System.ReactiveValue
     False Float=&FloatFromCurve(
      FloatCurve=DataInterface() {propertyValues {Curve {keys [
          {interpMode=RCIM_Cubic, leaveTangent=4.8905, tangentMode=RCTM_Break, time=0, value=0.15}
          <at 0.2 0.9>
          {interpMode=RCIM_Cubic, leaveTangent=-1.0585, tangentMode=RCTM_Break, time=0.3, value=0.9}
          {arriveTangent=-0.6609, tangentMode=RCTM_User, time=1, value=0.15}
         ]}}}
      CurveIndex=$System.NormalizedLoopAge
     )
    )
    Index 0 Param 4=&MakeCustomFloatFromBool(Bool=$System.bIsReactive, True Float=1, False Float=0)
    Module.Index0Param0Write=true
    Module.Index0Param1Write=true
    Module.Index0Param2Write=true
    Module.Index0Param3Write=true
   )
   &ScaleColor(
    Scale Mode=::NewEnumerator1
    Scale RGBA=&MakeVector4_FromLinearColor(LinearColor=&FNTimeOfDayColor(Night Color=<rgba 0.4683 0.4038 0.9 1>, Morning Color=<rgba 1 0.6536 0.6237 1>, Day Color=<rgba 1 1 1 1>, Evening Color=<rgba 1 1 1 1>))
    ScaleRGBA=true
   )
  ]
  rendererClasses [&NiagaraSpriteRendererProperties, &NiagaraRibbonRendererProperties]
  renderers [
   {rendererClass=&NiagaraSpriteRendererProperties}
   {rendererClass=&NiagaraRibbonRendererProperties, rendererIndex=1}
   {rendererClass=&NiagaraRibbonRendererProperties, rendererIndex=2}
   {rendererClass=&NiagaraRibbonRendererProperties, rendererIndex=3}
  ]
 }
 HelmetGlow {
  emitterUpdateScript [
   &EmitterState(Life Cycle Mode=::NewEnumerator1, Loop Behavior=::NewEnumerator1, Loop Duration=1.5, UseLoopDelay=false)
   &SpawnBurst_Instantaneous(Use Spawn Probability=false, UseLoopCountLimit=false) {module=SpawnBurst_Instantaneous001}
  ]
  particleSpawnScript [&InitializeParticle(
    Color=<rgba 1 1 1 0.7>
    Lifetime=5
    Sprite Size Mode=::NewEnumerator3
    Uniform Sprite Size=50
    UsePositionOffset=false
    Write Material Random=true
   )]
  particleUpdateScript [
   &ParticleState(Kill Particles When Lifetime Has Elapsed=false)
   &SkeletalMeshLocation(Bone Sampling Mode=::NewEnumerator4, MaskSpawnGroup=false, Skeletal Mesh=$System.BoneFlowSkeleMesh1)
   &ScaleColor(
    Scale Alpha=$System.AmplitudeValue
    Scale RGB=&MakeVectorFromLinearColorRGB(Color=&ColorFromCurve(
      ColorCurve=DataInterface() {propertyValues {
        AlphaCurve {keys [<at 0.4953 1 RCIM_Cubic>]}
        BlueCurve {keys [<at 0 0 RCIM_Cubic>,<at 0.1979 0>,<at 0.5342 1.7266>,<at 0.7607 0.7225>,<at 0.978 0>]}
        GreenCurve {keys [<at 0 0.4793 RCIM_Cubic>,<at 0.1979 2>,<at 0.5342 2>,<at 0.7607 0.1718>,<at 0.978 0.4793>]}
        RedCurve {keys [<at 0 1 RCIM_Cubic>,<at 0.1979 0.3323>,<at 0.5342 0>,<at 0.7607 1>,<at 0.978 1>]}
       }}
      CurveIndex=&Lerp_Float(A=$System.NormalizedLoopAge, B=$System.AmplitudeValue, Alpha=$System.Timeline.Time)
     ))
    ScaleA=true
    ScaleRGB=true
   )
  ]
  rendererClasses [&NiagaraSpriteRendererProperties]
  renderers [{rendererClass=&NiagaraSpriteRendererProperties}]
 }
 SplineGlow {
  emitterUpdateScript [
   &EmitterState(Life Cycle Mode=::NewEnumerator1, Loop Behavior=::NewEnumerator1, Loop Duration Mode=::NewEnumerator1, UseLoopDelay=false)
   &SpawnBurst_Instantaneous(Spawn Count=10, Use Spawn Probability=false, UseLoopCountLimit=false) {module=SpawnBurst_Instantaneous001}
  ]
  particleSpawnScript [
   &InitializeParticle(Color=<rgba 1 1 1 0.35>, Sprite Size Mode=::NewEnumerator3, Uniform Sprite Size=12, UsePositionOffset=false, Write Material Random=true)
   &PartitionParticles(Number Of Partitions=10)
   &CameraOffset(Camera Offset Amount=&SelectFloatFromArray(Direct Array Index=$Engine.QualityLevel, Float Selection Array=DataInterface() {propertyValues {FloatData [-1,-1,1,1,1]}}))
  ]
  particleUpdateScript [
   &ParticleState(
    Kill Particles When Lifetime Has Elapsed=false
    Let Infinitely Lived Particles Die When Emitter Deactivates=true
    Loop Particles Lifetime=true
   )
   &SkeletalMeshLocation(
    Bone / Socket Index=&SelectIntFromArray(
     Direct Array Index=$Particles.PartitionParticles.Partition
     Int Selection Array=DataInterface() {propertyValues {IntData [0,2,3,5,6,8,9,11,12,15]}}
    )
    MaskSpawnGroup=false
    Mesh Sampling Type=::NewEnumerator5
    Sampled Position Offset=&SelectVectorFromArray(
     Direct Array Index=$Particles.PartitionParticles.Partition
     Vector Selection Array=DataInterface() {propertyValues {FloatData [<0 -3 0>, <2 2 0>, <0 -3 -0.5>, <2 1 0>, <1.5 2 0>, <0.5 1 1>, <1 1 0.5>, <1 2 -1>, <2 0 0>, <0 -2 0>]}}
    )
    Skeletal Mesh=$System.BoneFlowSkeleMesh1
    Socket Sampling Mode=::NewEnumerator5
   )
   &ScaleColor(
    Scale Alpha=&SelectFloatFromArray(Direct Array Index=$Engine.QualityLevel, Float Selection Array=DataInterface() {propertyValues {FloatData [0.4,0.4,1.25,1.25,1.25]}})
    Scale RGB=&Lerp_Vector(
     A=&MakeVectorFromLinearColorRGB(Color=&ColorFromCurve(ColorCurve=DataInterface() {propertyValues {
         AlphaCurve {keys [<at 0 1 RCIM_Cubic>]}
         BlueCurve {keys [<at 0 0.4526 RCIM_Cubic>,<at 1 1 RCIM_Cubic>]}
         GreenCurve {keys [<at 0 0.8796 RCIM_Cubic>,<at 1 0.6308 RCIM_Cubic>]}
         RedCurve {keys [<at 0 0.4013 RCIM_Cubic>,<at 1 0.0957 RCIM_Cubic>]}
        }}))
     B=&MakeVectorFromLinearColorRGB(Color=&ColorFromCurve(ColorCurve=DataInterface() {propertyValues {
         AlphaCurve {keys [
           {interpMode=RCIM_Cubic, leaveTangentWeight=0.3333, time=0, value=1}
           {arriveTangentWeight=0.3333, interpMode=RCIM_Cubic, time=0.5009, value=1}
          ]}
         BlueCurve {keys [
           {interpMode=RCIM_Cubic, tangentMode=RCTM_User, time=0, value=0.1566}
           {interpMode=RCIM_Cubic, leaveTangentWeight=0.1194, tangentMode=RCTM_User, time=0.2136, value=0.1827}
           {
            arriveTangent=1.0596
            arriveTangentWeight=0.1474
            interpMode=RCIM_Cubic
            leaveTangent=1.0596
            leaveTangentWeight=0.0666
            time=0.4622
            value=0.7498
           }
           {arriveTangent=-0.3234, arriveTangentWeight=0.0666, interpMode=RCIM_Cubic, leaveTangent=-0.3234, time=0.7387, value=0.9393}
           {interpMode=RCIM_Cubic, tangentMode=RCTM_User, time=0.9608, value=0.5885}
          ]}
         GreenCurve {keys [
           {interpMode=RCIM_Cubic, tangentMode=RCTM_User, time=0, value=0.5781}
           {arriveTangent=0.5147, interpMode=RCIM_Cubic, leaveTangent=0.5147, leaveTangentWeight=0.1194, time=0.2136, value=0.8802}
           {
            arriveTangent=0.5147
            arriveTangentWeight=0.1474
            interpMode=RCIM_Cubic
            leaveTangent=0.5147
            leaveTangentWeight=0.0666
            time=0.4622
            value=0.1354
           }
           {arriveTangent=0.3932, arriveTangentWeight=0.0666, interpMode=RCIM_Cubic, leaveTangent=0.3932, time=0.7387, value=0.9583}
           {interpMode=RCIM_Cubic, tangentMode=RCTM_User, time=0.9608, value=0.3315}
          ]}
         RedCurve {keys [
           {interpMode=RCIM_Cubic, tangentMode=RCTM_User, time=0, value=0.1742}
           {arriveTangent=0.1081, interpMode=RCIM_Cubic, leaveTangent=0.1081, leaveTangentWeight=0.1194, time=0.2136, value=0.7775}
           {arriveTangentWeight=0.2294, interpMode=RCIM_Cubic, leaveTangentWeight=0.1036, tangentMode=RCTM_User, time=0.4622, value=1}
           {arriveTangent=-1.8423, arriveTangentWeight=0.0666, interpMode=RCIM_Cubic, leaveTangent=-1.8423, time=0.7387, value=0.254}
           {interpMode=RCIM_Cubic, tangentMode=RCTM_User, time=0.9608, value=0.0813}
          ]}
        }}))
     Alpha=$System.AmplitudeValue
    )
    ScaleA=true
    ScaleRGB=true
   )
  ]
  rendererClasses [&NiagaraSpriteRendererProperties]
  renderers [
   {rendererClass=&NiagaraSpriteRendererProperties}
   {rendererClass=&NiagaraSpriteRendererProperties, rendererIndex=1}
  ]
 }
}
summary {
 systemName=NS_RelayStickBounty_Skin
 emitters [
  {emitterName=HelmetGlow, rendererClasses [&NiagaraSpriteRendererProperties]}
  {emitterName=BodyLights, rendererClasses [&NiagaraLightRendererProperties]}
  {bEnabled=false, emitterName=ArmSplines001, rendererClasses [&NiagaraRibbonRendererProperties]}
  {emitterName=SplineGlow, rendererClasses [&NiagaraSpriteRendererProperties]}
  {bEnabled=false, emitterName=ArmSplines, rendererClasses [&NiagaraRibbonRendererProperties]}
  {
   emitterName=Bone_Flow_Shoulders
   rendererClasses [&NiagaraSpriteRendererProperties, &NiagaraRibbonRendererProperties]
  }
  {
   emitterName=Bone_Flow_Waist
   rendererClasses [&NiagaraSpriteRendererProperties, &NiagaraRibbonRendererProperties]
  }
  {
   emitterName=Bone_Flow_Hip_Right
   rendererClasses [&NiagaraSpriteRendererProperties, &NiagaraRibbonRendererProperties]
  }
  {
   emitterName=Bone_Flow_Hip_Left
   rendererClasses [&NiagaraSpriteRendererProperties, &NiagaraRibbonRendererProperties]
  }
 ]
 userVariables [
  User.bIsFrontEnd {type=NiagaraBool}
  User.bIsFrontEndPreview {type=NiagaraBool}
 ]
}
